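# Sound cue for the player chrome blob enter sound, varied by the PlayerAffliation parameter.
# A switch sends its first two slots to a boosted 1P foley class branch and the other two to separate 3P enemy foley class branches, one heavily attenuated by its mixer.
# Each branch picks evenly at random from the same four waves.
# Blob attenuation and a concurrency override limit range and stacking.
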SoundCue /ChromeShared/Audio/Player/Chrome_Blob_Enter_Cue
root=Switch_0
nodes [
 {id=Mixer_0, children [Random_1], properties {inputVolume [4]}}
 {id=Mixer_1, children [Random_5], properties {inputVolume [0.1]}}
 {id=Mixer_2, children [Random_6], properties {inputVolume [1]}}
 {
  id=Random_1
  children [WavePlayer_0,WavePlayer_1,WavePlayer_2,WavePlayer_7]
  properties {bShouldExcludeFromBranchCulling=true, weights [1,1,1,1]}
 }
 {
  id=Random_5
  children [WavePlayer_8, WavePlayer_9, WavePlayer_10, WavePlayer_11]
  properties {bShouldExcludeFromBranchCulling=true, weights [1,1,1,1]}
 }
 {
  id=Random_6
  children [WavePlayer_12, WavePlayer_13, WavePlayer_14, WavePlayer_15]
  properties {bShouldExcludeFromBranchCulling=true, weights [1,1,1,1]}
 }
 {id=SoundClass_0, children [Mixer_1], properties {soundClassOverride=&Fort_Player_Foley_3P_Enemy_Athena}}
 {id=SoundClass_1, children [Mixer_0], properties {soundClassOverride=&Fort_Player_Foley_1P_Athena}}
 {id=SoundClass_2, children [Mixer_2], properties {soundClassOverride=&Fort_Player_Foley_3P_Enemy_Athena}}
 {id=Switch_0, children [SoundClass_1,SoundClass_1,SoundClass_0,SoundClass_2], properties {intParameterName=PlayerAffliation}}
 {id=WavePlayer_0, properties {soundWaveAssetPtr=&/Chrome_Blob_Enter_01}}
 {id=WavePlayer_1, properties {soundWaveAssetPtr=&/Chrome_Blob_Enter_02}}
 {id=WavePlayer_10, properties {soundWaveAssetPtr=&/Chrome_Blob_Enter_03}}
 {id=WavePlayer_11, properties {soundWaveAssetPtr=&/Chrome_Blob_Enter_04}}
 {id=WavePlayer_12, properties {soundWaveAssetPtr=&Chrome_Blob_Enter_01}}
 {id=WavePlayer_13, properties {soundWaveAssetPtr=&Chrome_Blob_Enter_02}}
 {id=WavePlayer_14, properties {soundWaveAssetPtr=&Chrome_Blob_Enter_03}}
 {id=WavePlayer_15, properties {soundWaveAssetPtr=&Chrome_Blob_Enter_04}}
 {id=WavePlayer_2, properties {soundWaveAssetPtr=&Chrome_Blob_Enter_03}}
 {id=WavePlayer_7, properties {soundWaveAssetPtr=&Chrome_Blob_Enter_04}}
 {id=WavePlayer_8, properties {soundWaveAssetPtr=&Chrome_Blob_Enter_01}}
 {id=WavePlayer_9, properties {soundWaveAssetPtr=&Chrome_Blob_Enter_02}}
]
properties {
 attenuationSettings=&/Chrome_Player_Blob_Att
 bDebug=true
 bOverrideConcurrency=true
 maxDistance=8000
 volumeMultiplier=1
 concurrencyOverrides {bVolumeScaleCanRelease=true, maxCount=3, volumeScale=0.5, volumeScaleAttackTime=0.1}
}
