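# Sound cue that plays one of four random Chrome Blob exit waves when the player exits blob form.
# A PlayerAffliation switch selects between branches, each with its own sound class, mixer volume, modulator and random node.
# The first two inputs share a boosted first-person foley branch; the other two use enemy third-person foley, one heavily attenuated, one at unity.
# Blob attenuation and a two-voice concurrency override with volume scaling apply.
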SoundCue /ChromeShared/Audio/Player/Chrome_Blob_Exit_Cue
root=Switch_0
nodes [
 {id=Mixer_0, children [Modulator_1], properties {inputVolume [6]}}
 {id=Mixer_1, children [Modulator_2], properties {inputVolume [0.1]}}
 {id=Mixer_2, children [Modulator_3], properties {inputVolume [1]}}
 {id=Modulator_1, children [Random_2]}
 {id=Modulator_2, children [Random_3]}
 {id=Modulator_3, children [Random_4]}
 {
  id=Random_2
  children [WavePlayer_16, WavePlayer_17, WavePlayer_18, WavePlayer_19]
  properties {bShouldExcludeFromBranchCulling=true, weights [1,1,1,1]}
 }
 {
  id=Random_3
  children [WavePlayer_20, WavePlayer_21, WavePlayer_22, WavePlayer_23]
  properties {bShouldExcludeFromBranchCulling=true, weights [1,1,1,1]}
 }
 {
  id=Random_4
  children [WavePlayer_24, WavePlayer_25, WavePlayer_26, WavePlayer_27]
  properties {bShouldExcludeFromBranchCulling=true, weights [1,1,1,1]}
 }
 {id=SoundClass_0, children [Mixer_1], properties {soundClassOverride=&Fort_Player_Foley_3P_Enemy_Athena}}
 {id=SoundClass_1, children [Mixer_0], properties {soundClassOverride=&Fort_Player_Foley_1P_Athena}}
 {id=SoundClass_2, children [Mixer_2], properties {soundClassOverride=&Fort_Player_Foley_3P_Enemy_Athena}}
 {id=Switch_0, children [SoundClass_1,SoundClass_1,SoundClass_0,SoundClass_2], properties {intParameterName=PlayerAffliation}}
 {id=WavePlayer_16, properties {soundWaveAssetPtr=&/Chrome_Blob_Exit_01}}
 {id=WavePlayer_17, properties {soundWaveAssetPtr=&/Chrome_Blob_Exit_02}}
 {id=WavePlayer_18, properties {soundWaveAssetPtr=&/Chrome_Blob_Exit_03}}
 {id=WavePlayer_19, properties {soundWaveAssetPtr=&/Chrome_Blob_Exit_04}}
 {id=WavePlayer_20, properties {soundWaveAssetPtr=&Chrome_Blob_Exit_01}}
 {id=WavePlayer_21, properties {soundWaveAssetPtr=&Chrome_Blob_Exit_02}}
 {id=WavePlayer_22, properties {soundWaveAssetPtr=&Chrome_Blob_Exit_03}}
 {id=WavePlayer_23, properties {soundWaveAssetPtr=&Chrome_Blob_Exit_04}}
 {id=WavePlayer_24, properties {soundWaveAssetPtr=&Chrome_Blob_Exit_01}}
 {id=WavePlayer_25, properties {soundWaveAssetPtr=&Chrome_Blob_Exit_02}}
 {id=WavePlayer_26, properties {soundWaveAssetPtr=&Chrome_Blob_Exit_03}}
 {id=WavePlayer_27, properties {soundWaveAssetPtr=&Chrome_Blob_Exit_04}}
]
properties {
 attenuationSettings=&/Chrome_Player_Blob_Att
 bDebug=true
 bOverrideConcurrency=true
 maxDistance=8000
 volumeMultiplier=1
 concurrencyOverrides {bVolumeScaleCanRelease=true, maxCount=2, volumeScale=0.5, volumeScaleAttackTime=0.1}
}
