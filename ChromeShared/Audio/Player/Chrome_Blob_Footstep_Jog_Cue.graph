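# Sound cue for jogging footsteps of the Chrome blob player form, with attenuation and a small overridden concurrency limit.
# A PlayerAffliation switch selects a first-person foley branch or one of two third-person enemy foley branches, each with its own sound class and mixer volume.
# Each branch randomly picks among five equally weighted waves with pitch modulation.
# Quality-level nodes leave their last two slots empty.
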SoundCue /ChromeShared/Audio/Player/Chrome_Blob_Footstep_Jog_Cue
root=Switch_0
nodes [
 {id=Mixer_0, children [QualityLevel_1], properties {inputVolume [0.1]}}
 {id=Mixer_1, children [QualityLevel_2], properties {inputVolume [0.8]}}
 {id=Mixer_2, children [QualityLevel_0], properties {inputVolume [5]}}
 {id=Modulator_3, children [Random_2], properties {pitchMax=1.2, volumeMax=1}}
 {id=Modulator_7, children [Random_3], properties {pitchMax=1.2, volumeMax=1}}
 {id=Modulator_8, children [Random_4], properties {pitchMax=1.2, volumeMax=1}}
 {id=QualityLevel_0, children [Modulator_3,Modulator_3,null,null]}
 {id=QualityLevel_1, children [Modulator_7,Modulator_7,null,null]}
 {id=QualityLevel_2, children [Modulator_8,Modulator_8,null,null]}
 {id=Random_2, children [WavePlayer_16, WavePlayer_17, WavePlayer_18, WavePlayer_19, WavePlayer_20], properties {weights [1,1,1,1,1]}}
 {id=Random_3, children [WavePlayer_33, WavePlayer_34, WavePlayer_35, WavePlayer_36, WavePlayer_37], properties {weights [1,1,1,1,1]}}
 {id=Random_4, children [WavePlayer_38, WavePlayer_39, WavePlayer_40, WavePlayer_41, WavePlayer_42], properties {weights [1,1,1,1,1]}}
 {id=SoundClass_0, children [Mixer_0], properties {soundClassOverride=&Fort_Player_Foley_3P_Enemy_Athena}}
 {id=SoundClass_1, children [Mixer_2], properties {soundClassOverride=&Fort_Player_Foley_1P_Athena}}
 {id=SoundClass_2, children [Mixer_1], properties {soundClassOverride=&Fort_Player_Foley_3P_Enemy_Athena}}
 {id=Switch_0, children [SoundClass_1,SoundClass_1,SoundClass_0,SoundClass_2], properties {intParameterName=PlayerAffliation}}
 {id=WavePlayer_16, properties {soundWaveAssetPtr=&/Chrome_Blob_Footstep_Jog_01}}
 {id=WavePlayer_17, properties {soundWaveAssetPtr=&/Chrome_Blob_Footstep_Jog_02}}
 {id=WavePlayer_18, properties {soundWaveAssetPtr=&/Chrome_Blob_Footstep_Jog_03}}
 {id=WavePlayer_19, properties {soundWaveAssetPtr=&/Chrome_Blob_Footstep_Jog_04}}
 {id=WavePlayer_20, properties {soundWaveAssetPtr=&/Chrome_Blob_Footstep_Jog_05}}
 {id=WavePlayer_33, properties {soundWaveAssetPtr=&Chrome_Blob_Footstep_Jog_01}}
 {id=WavePlayer_34, properties {soundWaveAssetPtr=&Chrome_Blob_Footstep_Jog_02}}
 {id=WavePlayer_35, properties {soundWaveAssetPtr=&Chrome_Blob_Footstep_Jog_03}}
 {id=WavePlayer_36, properties {soundWaveAssetPtr=&Chrome_Blob_Footstep_Jog_04}}
 {id=WavePlayer_37, properties {soundWaveAssetPtr=&Chrome_Blob_Footstep_Jog_05}}
 {id=WavePlayer_38, properties {soundWaveAssetPtr=&Chrome_Blob_Footstep_Jog_01}}
 {id=WavePlayer_39, properties {soundWaveAssetPtr=&Chrome_Blob_Footstep_Jog_02}}
 {id=WavePlayer_40, properties {soundWaveAssetPtr=&Chrome_Blob_Footstep_Jog_03}}
 {id=WavePlayer_41, properties {soundWaveAssetPtr=&Chrome_Blob_Footstep_Jog_04}}
 {id=WavePlayer_42, properties {soundWaveAssetPtr=&Chrome_Blob_Footstep_Jog_05}}
]
properties {
 attenuationSettings=&/Chrome_Player_Blob_Att
 bDebug=true
 bOverrideConcurrency=true
 maxDistance=8000
 virtualizationMode=Disabled
 volumeMultiplier=1
 concurrencyOverrides {bVolumeScaleCanRelease=true, maxCount=3, volumeScale=0.5, volumeScaleAttackTime=0.1}
}
