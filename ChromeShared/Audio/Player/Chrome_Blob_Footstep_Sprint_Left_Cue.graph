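# Sound cue for the left sprint footstep of the Chrome blob player form.
# The PlayerAffliation parameter switches between one first-person foley branch (shared by two indices) and two third-person enemy-class branches with different mix levels, one using dedicated 3P plant waves.
# Each branch mixes a modulated random plant with a random swirl that only the first two quality levels play.
# The cue sets its own attenuation and concurrency limit and disables virtualization.
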SoundCue /ChromeShared/Audio/Player/Chrome_Blob_Footstep_Sprint_Left_Cue
root=Switch_1
nodes [
 {id=Mixer_1, children [Modulator_3, QualityLevel_1], properties {inputVolume [4,4]}}
 {id=Mixer_2, children [Modulator_5, QualityLevel_2], properties {inputVolume [0.1,0.1]}}
 {id=Mixer_3, children [Modulator_7, QualityLevel_3], properties {inputVolume [0.8,0.8]}}
 {id=Modulator_3, children [Random_2], properties {pitchMax=1.1, volumeMax=1.5}}
 {id=Modulator_4, children [Random_3]}
 {id=Modulator_5, children [Random_4]}
 {id=Modulator_6, children [Random_5]}
 {id=Modulator_7, children [Random_1]}
 {id=Modulator_8, children [Random_10]}
 {id=QualityLevel_1, children [Modulator_4,Modulator_4,null,null]}
 {id=QualityLevel_2, children [Modulator_6,Modulator_6,null,null]}
 {id=QualityLevel_3, children [Modulator_8,Modulator_8,null,null]}
 {
  id=Random_1
  children [
   WavePlayer_5
   WavePlayer_6
   WavePlayer_7
   WavePlayer_8
   WavePlayer_9
   WavePlayer_10
   WavePlayer_11
   WavePlayer_12
   WavePlayer_13
   WavePlayer_14
  ]
  properties {bShouldExcludeFromBranchCulling=true, weights [1,1,1,1,1,1,1,1,1,1]}
 }
 {
  id=Random_10
  children [
   WavePlayer_51
   WavePlayer_52
   WavePlayer_53
   WavePlayer_78
   WavePlayer_79
   WavePlayer_80
   WavePlayer_81
   WavePlayer_82
   WavePlayer_83
   WavePlayer_84
  ]
  properties {weights [1,1,1,1,1,1,1,1,1,1]}
 }
 {
  id=Random_2
  children [WavePlayer_16, WavePlayer_17, WavePlayer_18, WavePlayer_19, WavePlayer_20]
  properties {bShouldExcludeFromBranchCulling=true, weights [1,1,1,1,1]}
 }
 {
  id=Random_3
  children [
   WavePlayer_21
   WavePlayer_22
   WavePlayer_23
   WavePlayer_24
   WavePlayer_25
   WavePlayer_26
   WavePlayer_27
   WavePlayer_28
   WavePlayer_29
   WavePlayer_30
  ]
  properties {weights [1,1,1,1,1,1,1,1,1,1]}
 }
 {
  id=Random_4
  children [WavePlayer_31, WavePlayer_32, WavePlayer_33, WavePlayer_34, WavePlayer_35]
  properties {bShouldExcludeFromBranchCulling=true, weights [1,1,1,1,1]}
 }
 {
  id=Random_5
  children [
   WavePlayer_36
   WavePlayer_37
   WavePlayer_38
   WavePlayer_39
   WavePlayer_40
   WavePlayer_41
   WavePlayer_42
   WavePlayer_43
   WavePlayer_44
   WavePlayer_45
  ]
  properties {weights [1,1,1,1,1,1,1,1,1,1]}
 }
 {id=SoundClass_0, children [Mixer_3], properties {soundClassOverride=&Fort_Player_Foley_3P_Enemy_Athena}}
 {id=SoundClass_1, children [Mixer_1], properties {soundClassOverride=&Fort_Player_Foley_1P_Athena}}
 {id=SoundClass_2, children [Mixer_2], properties {soundClassOverride=&Fort_Player_Foley_3P_Enemy_Athena}}
 {id=Switch_1, children [SoundClass_1,SoundClass_1,SoundClass_2,SoundClass_0], properties {intParameterName=PlayerAffliation}}
 {id=WavePlayer_10, properties {soundWaveAssetPtr=&/Chrome_Blob_Footstep_Plant_3P_06}}
 {id=WavePlayer_11, properties {soundWaveAssetPtr=&/Chrome_Blob_Footstep_Plant_3P_07}}
 {id=WavePlayer_12, properties {soundWaveAssetPtr=&/Chrome_Blob_Footstep_Plant_3P_08}}
 {id=WavePlayer_13, properties {soundWaveAssetPtr=&/Chrome_Blob_Footstep_Plant_3P_09}}
 {id=WavePlayer_14, properties {soundWaveAssetPtr=&/Chrome_Blob_Footstep_Plant_3P_10}}
 {id=WavePlayer_16, properties {soundWaveAssetPtr=&/Chrome_Blob_Footstep_Plant_01}}
 {id=WavePlayer_17, properties {soundWaveAssetPtr=&/Chrome_Blob_Footstep_Plant_02}}
 {id=WavePlayer_18, properties {soundWaveAssetPtr=&/Chrome_Blob_Footstep_Plant_03}}
 {id=WavePlayer_19, properties {soundWaveAssetPtr=&/Chrome_Blob_Footstep_Plant_04}}
 {id=WavePlayer_20, properties {soundWaveAssetPtr=&/Chrome_Blob_Footstep_Plant_05}}
 {id=WavePlayer_21, properties {soundWaveAssetPtr=&/Chrome_Blob_Footstep_Swirl_01}}
 {id=WavePlayer_22, properties {soundWaveAssetPtr=&/Chrome_Blob_Footstep_Swirl_02}}
 {id=WavePlayer_23, properties {soundWaveAssetPtr=&/Chrome_Blob_Footstep_Swirl_03}}
 {id=WavePlayer_24, properties {soundWaveAssetPtr=&/Chrome_Blob_Footstep_Swirl_04}}
 {id=WavePlayer_25, properties {soundWaveAssetPtr=&/Chrome_Blob_Footstep_Swirl_05}}
 {id=WavePlayer_26, properties {soundWaveAssetPtr=&/Chrome_Blob_Footstep_Swirl_06}}
 {id=WavePlayer_27, properties {soundWaveAssetPtr=&/Chrome_Blob_Footstep_Swirl_07}}
 {id=WavePlayer_28, properties {soundWaveAssetPtr=&/Chrome_Blob_Footstep_Swirl_08}}
 {id=WavePlayer_29, properties {soundWaveAssetPtr=&/Chrome_Blob_Footstep_Swirl_09}}
 {id=WavePlayer_30, properties {soundWaveAssetPtr=&/Chrome_Blob_Footstep_Swirl_10}}
 {id=WavePlayer_31, properties {soundWaveAssetPtr=&Chrome_Blob_Footstep_Plant_01}}
 {id=WavePlayer_32, properties {soundWaveAssetPtr=&Chrome_Blob_Footstep_Plant_02}}
 {id=WavePlayer_33, properties {soundWaveAssetPtr=&Chrome_Blob_Footstep_Plant_03}}
 {id=WavePlayer_34, properties {soundWaveAssetPtr=&Chrome_Blob_Footstep_Plant_04}}
 {id=WavePlayer_35, properties {soundWaveAssetPtr=&Chrome_Blob_Footstep_Plant_05}}
 {id=WavePlayer_36, properties {soundWaveAssetPtr=&Chrome_Blob_Footstep_Swirl_01}}
 {id=WavePlayer_37, properties {soundWaveAssetPtr=&Chrome_Blob_Footstep_Swirl_02}}
 {id=WavePlayer_38, properties {soundWaveAssetPtr=&Chrome_Blob_Footstep_Swirl_03}}
 {id=WavePlayer_39, properties {soundWaveAssetPtr=&Chrome_Blob_Footstep_Swirl_04}}
 {id=WavePlayer_40, properties {soundWaveAssetPtr=&Chrome_Blob_Footstep_Swirl_05}}
 {id=WavePlayer_41, properties {soundWaveAssetPtr=&Chrome_Blob_Footstep_Swirl_06}}
 {id=WavePlayer_42, properties {soundWaveAssetPtr=&Chrome_Blob_Footstep_Swirl_07}}
 {id=WavePlayer_43, properties {soundWaveAssetPtr=&Chrome_Blob_Footstep_Swirl_08}}
 {id=WavePlayer_44, properties {soundWaveAssetPtr=&Chrome_Blob_Footstep_Swirl_09}}
 {id=WavePlayer_45, properties {soundWaveAssetPtr=&Chrome_Blob_Footstep_Swirl_10}}
 {id=WavePlayer_5, properties {soundWaveAssetPtr=&/Chrome_Blob_Footstep_Plant_3P_01}}
 {id=WavePlayer_51, properties {soundWaveAssetPtr=&Chrome_Blob_Footstep_Swirl_01}}
 {id=WavePlayer_52, properties {soundWaveAssetPtr=&Chrome_Blob_Footstep_Swirl_02}}
 {id=WavePlayer_53, properties {soundWaveAssetPtr=&Chrome_Blob_Footstep_Swirl_03}}
 {id=WavePlayer_6, properties {soundWaveAssetPtr=&/Chrome_Blob_Footstep_Plant_3P_02}}
 {id=WavePlayer_7, properties {soundWaveAssetPtr=&/Chrome_Blob_Footstep_Plant_3P_03}}
 {id=WavePlayer_78, properties {soundWaveAssetPtr=&Chrome_Blob_Footstep_Swirl_04}}
 {id=WavePlayer_79, properties {soundWaveAssetPtr=&Chrome_Blob_Footstep_Swirl_05}}
 {id=WavePlayer_8, properties {soundWaveAssetPtr=&/Chrome_Blob_Footstep_Plant_3P_04}}
 {id=WavePlayer_80, properties {soundWaveAssetPtr=&Chrome_Blob_Footstep_Swirl_06}}
 {id=WavePlayer_81, properties {soundWaveAssetPtr=&Chrome_Blob_Footstep_Swirl_07}}
 {id=WavePlayer_82, properties {soundWaveAssetPtr=&Chrome_Blob_Footstep_Swirl_08}}
 {id=WavePlayer_83, properties {soundWaveAssetPtr=&Chrome_Blob_Footstep_Swirl_09}}
 {id=WavePlayer_84, properties {soundWaveAssetPtr=&Chrome_Blob_Footstep_Swirl_10}}
 {id=WavePlayer_9, properties {soundWaveAssetPtr=&/Chrome_Blob_Footstep_Plant_3P_05}}
]
properties {
 attenuationSettings=&/Chrome_Player_Blob_Att
 bDebug=true
 bOverrideConcurrency=true
 maxDistance=8000
 virtualizationMode=Disabled
 volumeMultiplier=1
 concurrencyOverrides {bVolumeScaleCanRelease=true, maxCount=3, volumeScale=0.5, volumeScaleAttackTime=0.1}
}
