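# Sound Cue for the right sprint footstep of the Chrome Blob player form, with shared blob attenuation and a three-voice concurrency override.
# An integer switch on PlayerAffliation picks a loud 1P-class branch (first two slots), a very quiet enemy-3P-class branch, or an enemy-3P-class branch using 3P plant waves.
# Each branch mixes modulated random plant and swirl layers, with swirls wired only to the first two quality-level slots.
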SoundCue /ChromeShared/Audio/Player/Chrome_Blob_Footstep_Sprint_Right_Cue
root=Switch_0
nodes [
 {id=Mixer_1, children [Modulator_3, QualityLevel_1], properties {inputVolume [4,4]}}
 {id=Mixer_4, children [Modulator_0, QualityLevel_4], properties {inputVolume [0.1,0.1]}}
 {id=Mixer_5, children [Modulator_7, QualityLevel_5], properties {inputVolume [0.8,0.8]}}
 {id=Modulator_0, children [Random_7]}
 {id=Modulator_11, children [Random_9]}
 {id=Modulator_3, children [Random_2]}
 {id=Modulator_4, children [Random_3]}
 {id=Modulator_7, children [Random_1]}
 {id=Modulator_9, children [Random_8]}
 {id=QualityLevel_1, children [Modulator_4,Modulator_4,null,null]}
 {id=QualityLevel_4, children [Modulator_9,Modulator_9,null,null]}
 {id=QualityLevel_5, children [Modulator_11,Modulator_11,null,null]}
 {
  id=Random_1
  children [
   WavePlayer_5
   WavePlayer_6
   WavePlayer_7
   WavePlayer_8
   WavePlayer_9
   WavePlayer_10
   WavePlayer_11
   WavePlayer_12
   WavePlayer_13
   WavePlayer_14
  ]
  properties {bShouldExcludeFromBranchCulling=true, weights [1,1,1,1,1,1,1,1,1,1]}
 }
 {
  id=Random_2
  children [WavePlayer_16, WavePlayer_17, WavePlayer_18, WavePlayer_19, WavePlayer_20]
  properties {bShouldExcludeFromBranchCulling=true, weights [1,1,1,1,1]}
 }
 {
  id=Random_3
  children [
   WavePlayer_21
   WavePlayer_22
   WavePlayer_23
   WavePlayer_24
   WavePlayer_25
   WavePlayer_26
   WavePlayer_27
   WavePlayer_28
   WavePlayer_29
   WavePlayer_30
  ]
  properties {weights [1,1,1,1,1,1,1,1,1,1]}
 }
 {
  id=Random_7
  children [WavePlayer_0, WavePlayer_54, WavePlayer_55, WavePlayer_56, WavePlayer_57]
  properties {bShouldExcludeFromBranchCulling=true, weights [1,1,1,1,1]}
 }
 {
  id=Random_8
  children [
   WavePlayer_58
   WavePlayer_59
   WavePlayer_60
   WavePlayer_61
   WavePlayer_62
   WavePlayer_63
   WavePlayer_64
   WavePlayer_65
   WavePlayer_66
   WavePlayer_67
  ]
  properties {weights [1,1,1,1,1,1,1,1,1,1]}
 }
 {
  id=Random_9
  children [
   WavePlayer_68
   WavePlayer_69
   WavePlayer_70
   WavePlayer_71
   WavePlayer_72
   WavePlayer_73
   WavePlayer_74
   WavePlayer_75
   WavePlayer_76
   WavePlayer_77
  ]
  properties {weights [1,1,1,1,1,1,1,1,1,1]}
 }
 {id=SoundClass_0, children [Mixer_5], properties {soundClassOverride=&Fort_Player_Foley_3P_Enemy_Athena}}
 {id=SoundClass_1, children [Mixer_1], properties {soundClassOverride=&Fort_Player_Foley_1P_Athena}}
 {id=SoundClass_2, children [Mixer_4], properties {soundClassOverride=&Fort_Player_Foley_3P_Enemy_Athena}}
 {id=Switch_0, children [SoundClass_1,SoundClass_1,SoundClass_2,SoundClass_0], properties {intParameterName=PlayerAffliation}}
 {id=WavePlayer_0, properties {soundWaveAssetPtr=&/Chrome_Blob_Footstep_Plant_01}}
 {id=WavePlayer_10, properties {soundWaveAssetPtr=&/Chrome_Blob_Footstep_Plant_3P_06}}
 {id=WavePlayer_11, properties {soundWaveAssetPtr=&/Chrome_Blob_Footstep_Plant_3P_07}}
 {id=WavePlayer_12, properties {soundWaveAssetPtr=&/Chrome_Blob_Footstep_Plant_3P_08}}
 {id=WavePlayer_13, properties {soundWaveAssetPtr=&/Chrome_Blob_Footstep_Plant_3P_09}}
 {id=WavePlayer_14, properties {soundWaveAssetPtr=&/Chrome_Blob_Footstep_Plant_3P_10}}
 {id=WavePlayer_16, properties {soundWaveAssetPtr=&Chrome_Blob_Footstep_Plant_01}}
 {id=WavePlayer_17, properties {soundWaveAssetPtr=&/Chrome_Blob_Footstep_Plant_02}}
 {id=WavePlayer_18, properties {soundWaveAssetPtr=&/Chrome_Blob_Footstep_Plant_03}}
 {id=WavePlayer_19, properties {soundWaveAssetPtr=&/Chrome_Blob_Footstep_Plant_04}}
 {id=WavePlayer_20, properties {soundWaveAssetPtr=&/Chrome_Blob_Footstep_Plant_05}}
 {id=WavePlayer_21, properties {soundWaveAssetPtr=&/Chrome_Blob_Footstep_Swirl_01}}
 {id=WavePlayer_22, properties {soundWaveAssetPtr=&/Chrome_Blob_Footstep_Swirl_02}}
 {id=WavePlayer_23, properties {soundWaveAssetPtr=&/Chrome_Blob_Footstep_Swirl_03}}
 {id=WavePlayer_24, properties {soundWaveAssetPtr=&/Chrome_Blob_Footstep_Swirl_04}}
 {id=WavePlayer_25, properties {soundWaveAssetPtr=&/Chrome_Blob_Footstep_Swirl_05}}
 {id=WavePlayer_26, properties {soundWaveAssetPtr=&/Chrome_Blob_Footstep_Swirl_06}}
 {id=WavePlayer_27, properties {soundWaveAssetPtr=&/Chrome_Blob_Footstep_Swirl_07}}
 {id=WavePlayer_28, properties {soundWaveAssetPtr=&/Chrome_Blob_Footstep_Swirl_08}}
 {id=WavePlayer_29, properties {soundWaveAssetPtr=&/Chrome_Blob_Footstep_Swirl_09}}
 {id=WavePlayer_30, properties {soundWaveAssetPtr=&/Chrome_Blob_Footstep_Swirl_10}}
 {id=WavePlayer_5, properties {soundWaveAssetPtr=&/Chrome_Blob_Footstep_Plant_3P_01}}
 {id=WavePlayer_54, properties {soundWaveAssetPtr=&Chrome_Blob_Footstep_Plant_02}}
 {id=WavePlayer_55, properties {soundWaveAssetPtr=&Chrome_Blob_Footstep_Plant_03}}
 {id=WavePlayer_56, properties {soundWaveAssetPtr=&Chrome_Blob_Footstep_Plant_04}}
 {id=WavePlayer_57, properties {soundWaveAssetPtr=&Chrome_Blob_Footstep_Plant_05}}
 {id=WavePlayer_58, properties {soundWaveAssetPtr=&Chrome_Blob_Footstep_Swirl_01}}
 {id=WavePlayer_59, properties {soundWaveAssetPtr=&Chrome_Blob_Footstep_Swirl_02}}
 {id=WavePlayer_6, properties {soundWaveAssetPtr=&/Chrome_Blob_Footstep_Plant_3P_02}}
 {id=WavePlayer_60, properties {soundWaveAssetPtr=&Chrome_Blob_Footstep_Swirl_03}}
 {id=WavePlayer_61, properties {soundWaveAssetPtr=&Chrome_Blob_Footstep_Swirl_04}}
 {id=WavePlayer_62, properties {soundWaveAssetPtr=&Chrome_Blob_Footstep_Swirl_05}}
 {id=WavePlayer_63, properties {soundWaveAssetPtr=&Chrome_Blob_Footstep_Swirl_06}}
 {id=WavePlayer_64, properties {soundWaveAssetPtr=&Chrome_Blob_Footstep_Swirl_07}}
 {id=WavePlayer_65, properties {soundWaveAssetPtr=&Chrome_Blob_Footstep_Swirl_08}}
 {id=WavePlayer_66, properties {soundWaveAssetPtr=&Chrome_Blob_Footstep_Swirl_09}}
 {id=WavePlayer_67, properties {soundWaveAssetPtr=&Chrome_Blob_Footstep_Swirl_10}}
 {id=WavePlayer_68, properties {soundWaveAssetPtr=&Chrome_Blob_Footstep_Swirl_01}}
 {id=WavePlayer_69, properties {soundWaveAssetPtr=&Chrome_Blob_Footstep_Swirl_02}}
 {id=WavePlayer_7, properties {soundWaveAssetPtr=&/Chrome_Blob_Footstep_Plant_3P_03}}
 {id=WavePlayer_70, properties {soundWaveAssetPtr=&Chrome_Blob_Footstep_Swirl_03}}
 {id=WavePlayer_71, properties {soundWaveAssetPtr=&Chrome_Blob_Footstep_Swirl_04}}
 {id=WavePlayer_72, properties {soundWaveAssetPtr=&Chrome_Blob_Footstep_Swirl_05}}
 {id=WavePlayer_73, properties {soundWaveAssetPtr=&Chrome_Blob_Footstep_Swirl_06}}
 {id=WavePlayer_74, properties {soundWaveAssetPtr=&Chrome_Blob_Footstep_Swirl_07}}
 {id=WavePlayer_75, properties {soundWaveAssetPtr=&Chrome_Blob_Footstep_Swirl_08}}
 {id=WavePlayer_76, properties {soundWaveAssetPtr=&Chrome_Blob_Footstep_Swirl_09}}
 {id=WavePlayer_77, properties {soundWaveAssetPtr=&Chrome_Blob_Footstep_Swirl_10}}
 {id=WavePlayer_8, properties {soundWaveAssetPtr=&/Chrome_Blob_Footstep_Plant_3P_04}}
 {id=WavePlayer_9, properties {soundWaveAssetPtr=&/Chrome_Blob_Footstep_Plant_3P_05}}
]
properties {
 attenuationSettings=&/Chrome_Player_Blob_Att
 bDebug=true
 bOverrideConcurrency=true
 maxDistance=8000
 virtualizationMode=Disabled
 volumeMultiplier=1
 concurrencyOverrides {bVolumeScaleCanRelease=true, maxCount=3, volumeScale=0.5, volumeScaleAttackTime=0.1}
}
