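# Sound cue for the Chrome Blob player's walk footsteps, varied by the listener's relationship to the player.
# A PlayerAffliation switch routes to a first-person foley class or two third-person enemy-class branches, each branch mixed at its own volume.
# Each branch randomly picks among five step waves with pitch modulation, behind quality nodes that fill only the first two levels.
# Blob attenuation and a three-voice concurrency override with volume scaling limit playback.
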SoundCue /ChromeShared/Audio/Player/Chrome_Blob_Footstep_Walk_Cue
root=Switch_0
nodes [
 {id=Mixer_0, children [QualityLevel_1], properties {inputVolume [0.1]}}
 {id=Mixer_1, children [QualityLevel_2], properties {inputVolume [0.8]}}
 {id=Mixer_2, children [QualityLevel_0], properties {inputVolume [5]}}
 {id=Modulator_3, children [Random_0], properties {pitchMax=1.2, volumeMax=1}}
 {id=Modulator_7, children [Random_1], properties {pitchMax=1.2, volumeMax=1}}
 {id=Modulator_8, children [Random_5], properties {pitchMax=1.2, volumeMax=1}}
 {id=QualityLevel_0, children [Modulator_3,Modulator_3,null,null]}
 {id=QualityLevel_1, children [Modulator_7,Modulator_7,null,null]}
 {id=QualityLevel_2, children [Modulator_8,Modulator_8,null,null]}
 {
  id=Random_0
  children [WavePlayer_0,WavePlayer_1,WavePlayer_2,WavePlayer_3,WavePlayer_4]
  properties {bShouldExcludeFromBranchCulling=true, weights [1,1,1,1,1]}
 }
 {
  id=Random_1
  children [WavePlayer_5,WavePlayer_6,WavePlayer_7,WavePlayer_8,WavePlayer_9]
  properties {bShouldExcludeFromBranchCulling=true, weights [1,1,1,1,1]}
 }
 {
  id=Random_5
  children [WavePlayer_10, WavePlayer_11, WavePlayer_12, WavePlayer_13, WavePlayer_14]
  properties {bShouldExcludeFromBranchCulling=true, weights [1,1,1,1,1]}
 }
 {id=SoundClass_0, children [Mixer_0], properties {soundClassOverride=&Fort_Player_Foley_3P_Enemy_Athena}}
 {id=SoundClass_1, children [Mixer_2], properties {soundClassOverride=&Fort_Player_Foley_1P_Athena}}
 {id=SoundClass_2, children [Mixer_1], properties {soundClassOverride=&Fort_Player_Foley_3P_Enemy_Athena}}
 {id=Switch_0, children [SoundClass_1,SoundClass_1,SoundClass_0,SoundClass_2], properties {intParameterName=PlayerAffliation}}
 {id=WavePlayer_0, properties {soundWaveAssetPtr=&/Chrome_Blob_Footstep_Walk_01}}
 {id=WavePlayer_1, properties {soundWaveAssetPtr=&/Chrome_Blob_Footstep_Walk_02}}
 {id=WavePlayer_10, properties {soundWaveAssetPtr=&Chrome_Blob_Footstep_Walk_01}}
 {id=WavePlayer_11, properties {soundWaveAssetPtr=&Chrome_Blob_Footstep_Walk_02}}
 {id=WavePlayer_12, properties {soundWaveAssetPtr=&/Chrome_Blob_Footstep_Walk_03}}
 {id=WavePlayer_13, properties {soundWaveAssetPtr=&/Chrome_Blob_Footstep_Walk_04}}
 {id=WavePlayer_14, properties {soundWaveAssetPtr=&/Chrome_Blob_Footstep_Walk_05}}
 {id=WavePlayer_2, properties {soundWaveAssetPtr=&Chrome_Blob_Footstep_Walk_03}}
 {id=WavePlayer_3, properties {soundWaveAssetPtr=&Chrome_Blob_Footstep_Walk_04}}
 {id=WavePlayer_4, properties {soundWaveAssetPtr=&Chrome_Blob_Footstep_Walk_05}}
 {id=WavePlayer_5, properties {soundWaveAssetPtr=&Chrome_Blob_Footstep_Walk_01}}
 {id=WavePlayer_6, properties {soundWaveAssetPtr=&Chrome_Blob_Footstep_Walk_02}}
 {id=WavePlayer_7, properties {soundWaveAssetPtr=&Chrome_Blob_Footstep_Walk_03}}
 {id=WavePlayer_8, properties {soundWaveAssetPtr=&Chrome_Blob_Footstep_Walk_04}}
 {id=WavePlayer_9, properties {soundWaveAssetPtr=&Chrome_Blob_Footstep_Walk_05}}
]
properties {
 attenuationSettings=&/Chrome_Player_Blob_Att
 bDebug=true
 bOverrideConcurrency=true
 maxDistance=8000
 virtualizationMode=Disabled
 volumeMultiplier=1
 concurrencyOverrides {bVolumeScaleCanRelease=true, maxCount=3, volumeScale=0.5, volumeScaleAttackTime=0.1}
}
