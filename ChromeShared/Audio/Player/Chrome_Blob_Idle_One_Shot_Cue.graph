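# Sound cue for the Chrome player blob's idle one-shot bubbles, randomly picking one of eight waves with wide pitch and volume modulation.
# A switch on PlayerAffliation selects a boosted first-person foley branch or one of two third-person enemy foley branches, one heavily attenuated.
# Each branch has its own mixer, quality gate, modulator and random node, with only the first two quality slots filled.
# Concurrency blocks new instances beyond three.
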SoundCue /ChromeShared/Audio/Player/Chrome_Blob_Idle_One_Shot_Cue
root=Switch_1
nodes [
 {id=Mixer_0, children [QualityLevel_5], properties {inputVolume [0.1]}}
 {id=Mixer_1, children [QualityLevel_6], properties {inputVolume [1]}}
 {id=Mixer_2, children [QualityLevel_4], properties {inputVolume [6]}}
 {id=Modulator_0, children [Random_1], properties {pitchMax=2, pitchMin=0.7, volumeMax=1, volumeMin=0.1}}
 {id=Modulator_1, children [Random_2], properties {pitchMax=2, pitchMin=0.7, volumeMax=1, volumeMin=0.1}}
 {id=Modulator_10, children [Random_0], properties {pitchMax=2, pitchMin=0.7, volumeMax=1, volumeMin=0.1}}
 {id=QualityLevel_4, children [Modulator_10,Modulator_10,null,null]}
 {id=QualityLevel_5, children [Modulator_0,Modulator_0,null,null]}
 {id=QualityLevel_6, children [Modulator_1,Modulator_1,null,null]}
 {
  id=Random_0
  children [WavePlayer_54, WavePlayer_55, WavePlayer_56, WavePlayer_57, WavePlayer_58, WavePlayer_59, WavePlayer_60, WavePlayer_61]
  properties {weights [1,1,1,1,1,1,1,1]}
 }
 {
  id=Random_1
  children [WavePlayer_62, WavePlayer_63, WavePlayer_64, WavePlayer_65, WavePlayer_66, WavePlayer_67, WavePlayer_68, WavePlayer_69]
  properties {bShouldExcludeFromBranchCulling=true, weights [1,1,1,1,1,1,1,1]}
 }
 {
  id=Random_2
  children [WavePlayer_70, WavePlayer_71, WavePlayer_72, WavePlayer_73, WavePlayer_74, WavePlayer_75, WavePlayer_76, WavePlayer_77]
  properties {bShouldExcludeFromBranchCulling=true, weights [1,1,1,1,1,1,1,1]}
 }
 {id=SoundClass_0, children [Mixer_1], properties {soundClassOverride=&Fort_Player_Foley_3P_Enemy_Athena}}
 {id=SoundClass_1, children [Mixer_2], properties {soundClassOverride=&Fort_Player_Foley_1P_Athena}}
 {id=SoundClass_2, children [Mixer_0], properties {soundClassOverride=&Fort_Player_Foley_3P_Enemy_Athena}}
 {id=Switch_1, children [SoundClass_1,SoundClass_1,SoundClass_2,SoundClass_0], properties {intParameterName=PlayerAffliation}}
 {id=WavePlayer_54, properties {soundWaveAssetPtr=&/Chrome_Blob_Idle_Bubble_01}}
 {id=WavePlayer_55, properties {soundWaveAssetPtr=&/Chrome_Blob_Idle_Bubble_02}}
 {id=WavePlayer_56, properties {soundWaveAssetPtr=&/Chrome_Blob_Idle_Bubble_03}}
 {id=WavePlayer_57, properties {soundWaveAssetPtr=&/Chrome_Blob_Idle_Bubble_04}}
 {id=WavePlayer_58, properties {soundWaveAssetPtr=&/Chrome_Blob_Idle_Bubble_05}}
 {id=WavePlayer_59, properties {soundWaveAssetPtr=&/Chrome_Blob_Idle_Bubble_06}}
 {id=WavePlayer_60, properties {soundWaveAssetPtr=&/Chrome_Blob_Idle_Bubble_07}}
 {id=WavePlayer_61, properties {soundWaveAssetPtr=&/Chrome_Blob_Idle_Bubble_08}}
 {id=WavePlayer_62, properties {soundWaveAssetPtr=&Chrome_Blob_Idle_Bubble_01}}
 {id=WavePlayer_63, properties {soundWaveAssetPtr=&Chrome_Blob_Idle_Bubble_02}}
 {id=WavePlayer_64, properties {soundWaveAssetPtr=&Chrome_Blob_Idle_Bubble_03}}
 {id=WavePlayer_65, properties {soundWaveAssetPtr=&Chrome_Blob_Idle_Bubble_04}}
 {id=WavePlayer_66, properties {soundWaveAssetPtr=&Chrome_Blob_Idle_Bubble_05}}
 {id=WavePlayer_67, properties {soundWaveAssetPtr=&Chrome_Blob_Idle_Bubble_06}}
 {id=WavePlayer_68, properties {soundWaveAssetPtr=&Chrome_Blob_Idle_Bubble_07}}
 {id=WavePlayer_69, properties {soundWaveAssetPtr=&Chrome_Blob_Idle_Bubble_08}}
 {id=WavePlayer_70, properties {soundWaveAssetPtr=&Chrome_Blob_Idle_Bubble_01}}
 {id=WavePlayer_71, properties {soundWaveAssetPtr=&Chrome_Blob_Idle_Bubble_02}}
 {id=WavePlayer_72, properties {soundWaveAssetPtr=&Chrome_Blob_Idle_Bubble_03}}
 {id=WavePlayer_73, properties {soundWaveAssetPtr=&Chrome_Blob_Idle_Bubble_04}}
 {id=WavePlayer_74, properties {soundWaveAssetPtr=&Chrome_Blob_Idle_Bubble_05}}
 {id=WavePlayer_75, properties {soundWaveAssetPtr=&Chrome_Blob_Idle_Bubble_06}}
 {id=WavePlayer_76, properties {soundWaveAssetPtr=&Chrome_Blob_Idle_Bubble_07}}
 {id=WavePlayer_77, properties {soundWaveAssetPtr=&Chrome_Blob_Idle_Bubble_08}}
]
properties {
 attenuationSettings=&/Chrome_Player_Blob_Att
 bDebug=true
 bOverrideConcurrency=true
 maxDistance=8000
 priority=0
 volumeMultiplier=1
 concurrencyOverrides {maxCount=3, resolutionRule=PreventNew}
}
