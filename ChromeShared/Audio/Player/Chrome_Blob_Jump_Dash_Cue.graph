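# Sound cue for the Chrome blob jump dash, switching layers and sound class on the PlayerAffliation integer parameter.
# The first two slots share a first-person foley mix of dash plus squish, with squish wired to only two quality levels.
# The third slot plays a very quiet dash, and the fourth mixes dash with 3P squish, both as enemy foley.
# Layers use random waves with modulation, under shared attenuation and capped concurrency.
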SoundCue /ChromeShared/Audio/Player/Chrome_Blob_Jump_Dash_Cue
root=Switch_1
nodes [
 {id=Mixer_0, children [Modulator_0, QualityLevel_0], properties {inputVolume [7,6.7]}}
 {id=Mixer_1, children [Modulator_2,null], properties {inputVolume [0.05,0.1]}}
 {id=Mixer_2, children [Modulator_4,Modulator_6], properties {inputVolume [1.2,0.9]}}
 {id=Modulator_0, children [Random_0], properties {pitchMax=1.1}}
 {id=Modulator_1, children [Random_2], properties {pitchMax=1.5, pitchMin=0.9, volumeMax=1.1, volumeMin=0.7}}
 {id=Modulator_2, children [Random_1], properties {pitchMax=1.1}}
 {id=Modulator_4, children [Random_4], properties {pitchMax=1.1}}
 {id=Modulator_6, children [Random_6], properties {pitchMax=1.15}}
 {id=QualityLevel_0, children [Modulator_1,Modulator_1,null,null]}
 {
  id=Random_0
  children [WavePlayer_0,WavePlayer_1,WavePlayer_2,WavePlayer_7,WavePlayer_9]
  properties {bShouldExcludeFromBranchCulling=true, weights [1,1,1,1,1]}
 }
 {
  id=Random_1
  children [WavePlayer_3,WavePlayer_4,WavePlayer_5,WavePlayer_6,WavePlayer_8]
  properties {bShouldExcludeFromBranchCulling=true, weights [1,1,1,1,1]}
 }
 {
  id=Random_2
  children [WavePlayer_10, WavePlayer_11, WavePlayer_12, WavePlayer_13, WavePlayer_14]
  properties {bShouldExcludeFromBranchCulling=true, weights [1,1,1,1,1]}
 }
 {
  id=Random_4
  children [WavePlayer_20, WavePlayer_21, WavePlayer_22, WavePlayer_23, WavePlayer_24]
  properties {bShouldExcludeFromBranchCulling=true, weights [1,1,1,1,1]}
 }
 {
  id=Random_6
  children [WavePlayer_30, WavePlayer_31, WavePlayer_32, WavePlayer_33, WavePlayer_34]
  properties {bShouldExcludeFromBranchCulling=true, weights [1,1,1,1,1]}
 }
 {id=SoundClass_0, children [Mixer_1], properties {soundClassOverride=&Fort_Player_Foley_3P_Enemy_Athena}}
 {id=SoundClass_1, children [Mixer_0], properties {soundClassOverride=&Fort_Player_Foley_1P_Athena}}
 {id=SoundClass_2, children [Mixer_2], properties {soundClassOverride=&Fort_Player_Foley_3P_Enemy_Athena}}
 {id=Switch_1, children [SoundClass_1,SoundClass_1,SoundClass_0,SoundClass_2], properties {intParameterName=PlayerAffliation}}
 {id=WavePlayer_0, properties {soundWaveAssetPtr=&/Chrome_Blob_Jump_Dash_01}}
 {id=WavePlayer_1, properties {soundWaveAssetPtr=&/Chrome_Blob_Jump_Dash_02}}
 {id=WavePlayer_10, properties {soundWaveAssetPtr=&/Chrome_Blob_Jump_Dash_Squish_05}}
 {id=WavePlayer_11, properties {soundWaveAssetPtr=&/Chrome_Blob_Jump_Dash_Squish_01}}
 {id=WavePlayer_12, properties {soundWaveAssetPtr=&/Chrome_Blob_Jump_Dash_Squish_02}}
 {id=WavePlayer_13, properties {soundWaveAssetPtr=&/Chrome_Blob_Jump_Dash_Squish_03}}
 {id=WavePlayer_14, properties {soundWaveAssetPtr=&/Chrome_Blob_Jump_Dash_Squish_04}}
 {id=WavePlayer_2, properties {soundWaveAssetPtr=&/Chrome_Blob_Jump_Dash_03}}
 {id=WavePlayer_20, properties {soundWaveAssetPtr=&Chrome_Blob_Jump_Dash_01}}
 {id=WavePlayer_21, properties {soundWaveAssetPtr=&Chrome_Blob_Jump_Dash_02}}
 {id=WavePlayer_22, properties {soundWaveAssetPtr=&Chrome_Blob_Jump_Dash_03}}
 {id=WavePlayer_23, properties {soundWaveAssetPtr=&/Chrome_Blob_Jump_Dash_04}}
 {id=WavePlayer_24, properties {soundWaveAssetPtr=&/Chrome_Blob_Jump_Dash_05}}
 {id=WavePlayer_3, properties {soundWaveAssetPtr=&Chrome_Blob_Jump_Dash_01}}
 {id=WavePlayer_30, properties {soundWaveAssetPtr=&/Chrome_Blob_Jump_Dash_Squish_3P_01}}
 {id=WavePlayer_31, properties {soundWaveAssetPtr=&/Chrome_Blob_Jump_Dash_Squish_3P_02}}
 {id=WavePlayer_32, properties {soundWaveAssetPtr=&/Chrome_Blob_Jump_Dash_Squish_3P_03}}
 {id=WavePlayer_33, properties {soundWaveAssetPtr=&/Chrome_Blob_Jump_Dash_Squish_3P_04}}
 {id=WavePlayer_34, properties {soundWaveAssetPtr=&/Chrome_Blob_Jump_Dash_Squish_3P_05}}
 {id=WavePlayer_4, properties {soundWaveAssetPtr=&Chrome_Blob_Jump_Dash_02}}
 {id=WavePlayer_5, properties {soundWaveAssetPtr=&Chrome_Blob_Jump_Dash_03}}
 {id=WavePlayer_6, properties {soundWaveAssetPtr=&Chrome_Blob_Jump_Dash_04}}
 {id=WavePlayer_7, properties {soundWaveAssetPtr=&Chrome_Blob_Jump_Dash_04}}
 {id=WavePlayer_8, properties {soundWaveAssetPtr=&Chrome_Blob_Jump_Dash_05}}
 {id=WavePlayer_9, properties {soundWaveAssetPtr=&Chrome_Blob_Jump_Dash_05}}
]
properties {
 attenuationSettings=&/Chrome_Player_Blob_Att
 bDebug=true
 bOverrideConcurrency=true
 maxDistance=8000
 soundClassObject=&Fort_WeaponMelee_PSM
 volumeMultiplier=1
 concurrencyOverrides {maxCount=3}
}
