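# Sound cue for the player's Chrome blob landing, routed through the first-person player foley sound class.
# A Random node picks one of four equally weighted landing waves and feeds a Modulator at default settings for pitch and volume variation.
# The cue boosts volume, uses the blob player attenuation, and overrides concurrency to cap simultaneous instances, stopping the lowest priority.
# Its debug flag is left on.
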SoundCue /ChromeShared/Audio/Player/Chrome_Blob_Land_Cue
root=Modulator_0
nodes [
 {id=Modulator_0, children [Random_0]}
 {
  id=Random_0
  children [WavePlayer_0,WavePlayer_1,WavePlayer_2,WavePlayer_7]
  properties {bShouldExcludeFromBranchCulling=true, weights [1,1,1,1]}
 }
 {id=WavePlayer_0, properties {soundWaveAssetPtr=&/Chrome_Blob_Land_01}}
 {id=WavePlayer_1, properties {soundWaveAssetPtr=&/Chrome_Blob_Land_02}}
 {id=WavePlayer_2, properties {soundWaveAssetPtr=&/Chrome_Blob_Land_03}}
 {id=WavePlayer_7, properties {soundWaveAssetPtr=&/Chrome_Blob_Land_04}}
]
properties {
 attenuationSettings=&/Chrome_Player_Blob_Att
 bDebug=true
 bOverrideConcurrency=true
 maxDistance=8000
 soundClassObject=&Fort_Player_Foley_1P_Athena
 volumeMultiplier=3
 concurrencyOverrides {maxCount=3, resolutionRule=StopLowestPriority}
}
