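# Sound cue for a hard landing of the Chrome player blob, routed to the first-person player foley sound class.
# A random node picks one of four equally weighted heavy landing waves and feeds a modulator that varies pitch and volume on each play.
# It uses the blob's player attenuation asset, boosts volume, and overrides concurrency to cap simultaneous instances, stopping the lowest priority.
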SoundCue /ChromeShared/Audio/Player/Chrome_Blob_Land_Hard_Cue
root=Modulator_0
nodes [
 {id=Modulator_0, children [Random_1]}
 {
  id=Random_1
  children [WavePlayer_3,WavePlayer_4,WavePlayer_5,WavePlayer_6]
  properties {bShouldExcludeFromBranchCulling=true, weights [1,1,1,1]}
 }
 {id=WavePlayer_3, properties {soundWaveAssetPtr=&/Chrome_Blob_Land_Heavy_01}}
 {id=WavePlayer_4, properties {soundWaveAssetPtr=&/Chrome_Blob_Land_Heavy_02}}
 {id=WavePlayer_5, properties {soundWaveAssetPtr=&/Chrome_Blob_Land_Heavy_03}}
 {id=WavePlayer_6, properties {soundWaveAssetPtr=&/Chrome_Blob_Land_Heavy_04}}
]
properties {
 attenuationSettings=&/Chrome_Player_Blob_Att
 bDebug=true
 bOverrideConcurrency=true
 maxDistance=8000
 soundClassObject=&Fort_Player_Foley_1P_Athena
 volumeMultiplier=4
 concurrencyOverrides {maxCount=3, resolutionRule=StopLowestPriority}
}
