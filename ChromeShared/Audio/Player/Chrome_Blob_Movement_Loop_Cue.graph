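# Looping sound cue for the Chrome blob player's movement, switched on the PlayerAffliation parameter between first-person, quiet third-person and enemy third-person branches.
# Each branch crossfades idle, slow, medium and fast loops on BlobMovement, which also modulates each loop's pitch.
# The louder enemy branch swaps in an enemy-specific fast loop and adds pitch modulation from BlobVerticality.
# Sound class overrides route the branches to player foley classes, with shared attenuation and a concurrency limit.
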SoundCue /ChromeShared/Audio/Player/Chrome_Blob_Movement_Loop_Cue
root=Switch_0
nodes [
 {id=Mixer_1, children [Modulator_3], properties {inputVolume [1.2]}}
 {id=Mixer_2, children [Modulator_0], properties {inputVolume [0.05]}}
 {id=Mixer_3, children [Modulator_1], properties {inputVolume [2.75]}}
 {
  id=ModulatorContinuous_0
  children [WavePlayer_5]
  properties {pitchModulationParams {minOutput=0.9, parameterName=BlobMovement}, volumeModulationParams {minOutput=1, parameterName=BlobMovement}}
 }
 {
  id=ModulatorContinuous_1
  children [WavePlayer_3]
  properties {
   pitchModulationParams {maxInput=0.5, minInput=0.01, minOutput=0.75, parameterName=BlobMovement}
   volumeModulationParams {minOutput=1, parameterName=BlobMovement}
  }
 }
 {
  id=ModulatorContinuous_12
  children [WavePlayer_14]
  properties {pitchModulationParams {minOutput=0.9, parameterName=BlobMovement}, volumeModulationParams {minOutput=1, parameterName=BlobMovement}}
 }
 {
  id=ModulatorContinuous_13
  children [WavePlayer_12]
  properties {
   pitchModulationParams {maxInput=0.5, minInput=0.01, minOutput=0.75, parameterName=BlobMovement}
   volumeModulationParams {minOutput=1, parameterName=BlobMovement}
  }
 }
 {
  id=ModulatorContinuous_14
  children [WavePlayer_15]
  properties {
   pitchModulationParams {minInput=0.5, minOutput=0.7, parameterName=BlobMovement}
   volumeModulationParams {minOutput=1, parameterName=BlobMovement}
  }
 }
 {
  id=ModulatorContinuous_15
  children [WavePlayer_13]
  properties {pitchModulationParams {minOutput=0.3, parameterName=BlobMovement}, volumeModulationParams {minOutput=1, parameterName=BlobMovement}}
 }
 {
  id=ModulatorContinuous_2
  children [WavePlayer_6]
  properties {
   pitchModulationParams {minInput=0.5, minOutput=0.7, parameterName=BlobMovement}
   volumeModulationParams {maxOutput=0.7, minOutput=1, parameterName=BlobMovement}
  }
 }
 {
  id=ModulatorContinuous_3
  children [WavePlayer_4]
  properties {pitchModulationParams {minOutput=0.9, parameterName=BlobMovement}, volumeModulationParams {minOutput=1, parameterName=BlobMovement}}
 }
 {
  id=ModulatorContinuous_4
  children [WavePlayer_2]
  properties {pitchModulationParams {minOutput=0.9, parameterName=BlobMovement}, volumeModulationParams {minOutput=1, parameterName=BlobMovement}}
 }
 {
  id=ModulatorContinuous_5
  children [WavePlayer_0]
  properties {
   pitchModulationParams {maxInput=0.5, minInput=0.01, minOutput=0.75, parameterName=BlobMovement}
   volumeModulationParams {minOutput=1, parameterName=BlobMovement}
  }
 }
 {
  id=ModulatorContinuous_6
  children [WavePlayer_7]
  properties {
   pitchModulationParams {minInput=0.5, minOutput=0.7, parameterName=BlobMovement}
   volumeModulationParams {minOutput=1, parameterName=BlobMovement}
  }
 }
 {
  id=ModulatorContinuous_7
  children [WavePlayer_1]
  properties {pitchModulationParams {minOutput=0.9, parameterName=BlobMovement}, volumeModulationParams {minOutput=1, parameterName=BlobMovement}}
 }
 {
  id=ModulatorContinuous_9
  children [Mixer_1]
  properties {pitchModulationParams {maxOutput=1.3, minInput=-1, minOutput=0.8, parameterName=BlobVerticality}}
 }
 {id=Modulator_0, children [ParamCrossFade_1], properties {volumeMax=1, volumeMin=1}}
 {id=Modulator_1, children [ParamCrossFade_0], properties {volumeMax=1, volumeMin=1}}
 {id=Modulator_3, children [ParamCrossFade_3], properties {volumeMax=1, volumeMin=1}}
 {
  id=ParamCrossFade_0
  children [ModulatorContinuous_0, ModulatorContinuous_1, ModulatorContinuous_2, ModulatorContinuous_3]
  properties {
   paramName=BlobMovement
   crossFadeInput [
    {fadeInDistanceEnd=0, fadeInDistanceStart=0, fadeOutDistanceEnd=0.2, fadeOutDistanceStart=0.1, volume=1}
    {fadeInDistanceEnd=0.2, fadeInDistanceStart=0.01, fadeOutDistanceEnd=0.7, fadeOutDistanceStart=0.5, volume=1}
    {fadeInDistanceEnd=0.8, fadeInDistanceStart=0.15, fadeOutDistanceEnd=100, fadeOutDistanceStart=100, volume=1}
    {fadeInDistanceEnd=1, fadeInDistanceStart=0.3, fadeOutDistanceEnd=100, fadeOutDistanceStart=100, volume=1.5}
   ]
  }
 }
 {
  id=ParamCrossFade_1
  children [ModulatorContinuous_4, ModulatorContinuous_5, ModulatorContinuous_6, ModulatorContinuous_7]
  properties {
   paramName=BlobMovement
   crossFadeInput [
    {fadeInDistanceEnd=0, fadeInDistanceStart=0, fadeOutDistanceEnd=0.2, fadeOutDistanceStart=0.1, volume=1}
    {fadeInDistanceEnd=0.2, fadeInDistanceStart=0.01, fadeOutDistanceEnd=0.7, fadeOutDistanceStart=0.5, volume=1}
    {fadeInDistanceEnd=0.8, fadeInDistanceStart=0.15, fadeOutDistanceEnd=100, fadeOutDistanceStart=100, volume=1}
    {fadeInDistanceEnd=1, fadeInDistanceStart=0.3, fadeOutDistanceEnd=100, fadeOutDistanceStart=100, volume=1.5}
   ]
  }
 }
 {
  id=ParamCrossFade_3
  children [ModulatorContinuous_12, ModulatorContinuous_13, ModulatorContinuous_14, ModulatorContinuous_15]
  properties {
   paramName=BlobMovement
   crossFadeInput [
    {fadeInDistanceEnd=0, fadeInDistanceStart=0, fadeOutDistanceEnd=0.2, fadeOutDistanceStart=0.1, volume=1}
    {fadeInDistanceEnd=0.2, fadeInDistanceStart=0.01, fadeOutDistanceEnd=0.7, fadeOutDistanceStart=0.5, volume=1}
    {fadeInDistanceEnd=0.8, fadeInDistanceStart=0.15, fadeOutDistanceEnd=100, fadeOutDistanceStart=100, volume=1}
    {fadeInDistanceEnd=1, fadeInDistanceStart=0.3, fadeOutDistanceEnd=100, fadeOutDistanceStart=0.5, volume=1.5}
   ]
  }
 }
 {id=SoundClass_0, children [Mixer_2], properties {soundClassOverride=&Fort_Player_Foley_3P_Enemy_Athena}}
 {id=SoundClass_1, children [Mixer_3], properties {soundClassOverride=&Fort_Player_Foley_1P_Athena}}
 {
  id=SoundClass_2
  children [ModulatorContinuous_9]
  properties {soundClassOverride=&Fort_Player_Foley_3P_Enemy_Athena}
 }
 {id=Switch_0, children [SoundClass_1,SoundClass_1,SoundClass_0,SoundClass_2], properties {intParameterName=PlayerAffliation}}
 {id=WavePlayer_0, properties {bLooping=true, soundWaveAssetPtr=&/Chrome_Blob_Move_Loop_Slow_01}}
 {id=WavePlayer_1, properties {bLooping=true, soundWaveAssetPtr=&/Chrome_Blob_Move_Loop_Fast_01}}
 {id=WavePlayer_12, properties {bLooping=true, soundWaveAssetPtr=&Chrome_Blob_Move_Loop_Slow_01}}
 {id=WavePlayer_13, properties {bLooping=true, soundWaveAssetPtr=&/Chrome_Blob_Move_Loop_Fast_Enemy_01}}
 {id=WavePlayer_14, properties {bLooping=true, soundWaveAssetPtr=&/Chrome_Blob_Move_Loop_Idle_01}}
 {id=WavePlayer_15, properties {bLooping=true, soundWaveAssetPtr=&/Chrome_Blob_Move_Loop_Medium_01}}
 {id=WavePlayer_2, properties {bLooping=true, soundWaveAssetPtr=&Chrome_Blob_Move_Loop_Idle_01}}
 {id=WavePlayer_3, properties {bLooping=true, soundWaveAssetPtr=&Chrome_Blob_Move_Loop_Slow_01}}
 {id=WavePlayer_4, properties {bLooping=true, soundWaveAssetPtr=&Chrome_Blob_Move_Loop_Fast_01}}
 {id=WavePlayer_5, properties {bLooping=true, soundWaveAssetPtr=&Chrome_Blob_Move_Loop_Idle_01}}
 {id=WavePlayer_6, properties {bLooping=true, soundWaveAssetPtr=&Chrome_Blob_Move_Loop_Medium_01}}
 {id=WavePlayer_7, properties {bLooping=true, soundWaveAssetPtr=&Chrome_Blob_Move_Loop_Medium_01}}
]
properties {
 attenuationSettings=&/Chrome_Player_Blob_Movement_Loop_Att
 bDebug=true
 bOverrideConcurrency=true
 maxDistance=6000
 volumeMultiplier=1
 concurrencyOverrides {maxCount=4}
}
