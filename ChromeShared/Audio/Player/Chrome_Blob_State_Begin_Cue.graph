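# Sound cue for the start of the Chrome player blob state, layering a random activation whoosh with a random state-begin sound.
# A PlayerAffliation switch sends its first two slots to one loud first-person foley mix and its last two to separate third-person enemy foley mixes, the third very quiet, the fourth moderate.
# Each layer has its own modulator, while blob attenuation and a two-voice concurrency override with volume scaling govern playback.
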SoundCue /ChromeShared/Audio/Player/Chrome_Blob_State_Begin_Cue
root=Switch_0
nodes [
 {id=Mixer_5, children [Modulator_5,Modulator_4], properties {inputVolume [5,4.7]}}
 {id=Mixer_6, children [Modulator_7,Modulator_6], properties {inputVolume [0.1,0.05]}}
 {id=Mixer_7, children [Modulator_9,Modulator_8], properties {inputVolume [1,0.7]}}
 {id=Modulator_4, children [Random_8]}
 {id=Modulator_5, children [Random_7]}
 {id=Modulator_6, children [Random_10]}
 {id=Modulator_7, children [Random_9]}
 {id=Modulator_8, children [Random_12]}
 {id=Modulator_9, children [Random_11]}
 {id=Random_10, children [WavePlayer_36, WavePlayer_37, WavePlayer_38, WavePlayer_39], properties {weights [1,1,1,1]}}
 {id=Random_11, children [WavePlayer_40, WavePlayer_41, WavePlayer_42], properties {weights [1,1,1]}}
 {id=Random_12, children [WavePlayer_43, WavePlayer_44, WavePlayer_45, WavePlayer_46], properties {weights [1,1,1,1]}}
 {id=Random_7, children [WavePlayer_26, WavePlayer_27, WavePlayer_28], properties {weights [1,1,1]}}
 {id=Random_8, children [WavePlayer_29, WavePlayer_30, WavePlayer_31, WavePlayer_32], properties {weights [1,1,1,1]}}
 {id=Random_9, children [WavePlayer_33, WavePlayer_34, WavePlayer_35], properties {weights [1,1,1]}}
 {id=SoundClass_0, children [Mixer_6], properties {soundClassOverride=&Fort_Player_Foley_3P_Enemy_Athena}}
 {id=SoundClass_1, children [Mixer_5], properties {soundClassOverride=&Fort_Player_Foley_1P_Athena}}
 {id=SoundClass_2, children [Mixer_7], properties {soundClassOverride=&Fort_Player_Foley_3P_Enemy_Athena}}
 {id=Switch_0, children [SoundClass_1,SoundClass_1,SoundClass_0,SoundClass_2], properties {intParameterName=PlayerAffliation}}
 {id=WavePlayer_26, properties {soundWaveAssetPtr=&/Chrome_Blob_Activate_Whoosh_01}}
 {id=WavePlayer_27, properties {soundWaveAssetPtr=&/Chrome_Blob_Activate_Whoosh_02}}
 {id=WavePlayer_28, properties {soundWaveAssetPtr=&/Chrome_Blob_Activate_Whoosh_03}}
 {id=WavePlayer_29, properties {soundWaveAssetPtr=&/Chrome_Blob_State_Begin_01}}
 {id=WavePlayer_30, properties {soundWaveAssetPtr=&/Chrome_Blob_State_Begin_02}}
 {id=WavePlayer_31, properties {soundWaveAssetPtr=&/Chrome_Blob_State_Begin_03}}
 {id=WavePlayer_32, properties {soundWaveAssetPtr=&/Chrome_Blob_State_Begin_04}}
 {id=WavePlayer_33, properties {soundWaveAssetPtr=&Chrome_Blob_Activate_Whoosh_01}}
 {id=WavePlayer_34, properties {soundWaveAssetPtr=&Chrome_Blob_Activate_Whoosh_02}}
 {id=WavePlayer_35, properties {soundWaveAssetPtr=&Chrome_Blob_Activate_Whoosh_03}}
 {id=WavePlayer_36, properties {soundWaveAssetPtr=&Chrome_Blob_State_Begin_01}}
 {id=WavePlayer_37, properties {soundWaveAssetPtr=&Chrome_Blob_State_Begin_02}}
 {id=WavePlayer_38, properties {soundWaveAssetPtr=&Chrome_Blob_State_Begin_03}}
 {id=WavePlayer_39, properties {soundWaveAssetPtr=&Chrome_Blob_State_Begin_04}}
 {id=WavePlayer_40, properties {soundWaveAssetPtr=&Chrome_Blob_Activate_Whoosh_01}}
 {id=WavePlayer_41, properties {soundWaveAssetPtr=&Chrome_Blob_Activate_Whoosh_02}}
 {id=WavePlayer_42, properties {soundWaveAssetPtr=&Chrome_Blob_Activate_Whoosh_03}}
 {id=WavePlayer_43, properties {soundWaveAssetPtr=&Chrome_Blob_State_Begin_01}}
 {id=WavePlayer_44, properties {soundWaveAssetPtr=&Chrome_Blob_State_Begin_02}}
 {id=WavePlayer_45, properties {soundWaveAssetPtr=&Chrome_Blob_State_Begin_03}}
 {id=WavePlayer_46, properties {soundWaveAssetPtr=&Chrome_Blob_State_Begin_04}}
]
properties {
 attenuationSettings=&/Chrome_Player_Blob_Att
 bDebug=true
 bOverrideConcurrency=true
 maxDistance=8000
 volumeMultiplier=1
 concurrencyOverrides {bVolumeScaleCanRelease=true, maxCount=2, volumeScale=0.5, volumeScaleAttackTime=0.1}
}
