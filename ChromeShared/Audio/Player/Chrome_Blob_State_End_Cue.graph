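# Sound cue named for the end of the player's Chrome blob state, currently empty with no root or sound nodes, so it has nothing to play.
# It still carries the player blob attenuation, a melee weapon sound class, and a concurrency override that stops the lowest-priority instance when the limit is reached.
# Its debug flag is left on.
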SoundCue /ChromeShared/Audio/Player/Chrome_Blob_State_End_Cue
root=null
nodes []
properties {
 attenuationSettings=&/Chrome_Player_Blob_Att
 bDebug=true
 bOverrideConcurrency=true
 maxDistance=4000
 soundClassObject=&Fort_WeaponMelee_PSM
 volumeMultiplier=1
 concurrencyOverrides {maxCount=3, resolutionRule=StopLowestPriority}
}
