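# Sound cue for the Chrome blob player's stop sound, with separate local and non-local mixes.
# An IsLocal branch picks a louder first-person foley class path, or an enemy third-person path when false or unset.
# Each path randomly picks from the same thirteen stop waves with pitch and volume modulation, leaving the fourth quality-level slot empty.
# The cue uses a dedicated attenuation asset and overrides concurrency to limit simultaneous plays.
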SoundCue /ChromeShared/Audio/Player/Chrome_Blob_Stop_Cue
root=Branch_0
nodes [
 {id=Branch_0, children [SoundClass_1,SoundClass_0,SoundClass_0], properties {boolParameterName=IsLocal}}
 {id=Mixer_1, children [QualityLevel_0], properties {inputVolume [3]}}
 {id=Mixer_2, children [QualityLevel_1], properties {inputVolume [1]}}
 {id=Modulator_0, children [Random_0], properties {pitchMax=1.2, volumeMax=1, volumeMin=0.6}}
 {id=Modulator_1, children [Random_1], properties {pitchMax=1.2, volumeMax=1, volumeMin=0.6}}
 {id=QualityLevel_0, children [Modulator_0,Modulator_0,Modulator_0,null]}
 {id=QualityLevel_1, children [Modulator_1,Modulator_1,Modulator_1,null]}
 {
  id=Random_0
  children [
   WavePlayer_1
   WavePlayer_2
   WavePlayer_3
   WavePlayer_4
   WavePlayer_5
   WavePlayer_6
   WavePlayer_7
   WavePlayer_8
   WavePlayer_9
   WavePlayer_10
   WavePlayer_11
   WavePlayer_12
   WavePlayer_13
  ]
  properties {bShouldExcludeFromBranchCulling=true, weights [1,1,1,1,1,1,1,1,1,1,1,1,1]}
 }
 {
  id=Random_1
  children [
   WavePlayer_14
   WavePlayer_15
   WavePlayer_16
   WavePlayer_17
   WavePlayer_18
   WavePlayer_19
   WavePlayer_20
   WavePlayer_21
   WavePlayer_22
   WavePlayer_23
   WavePlayer_24
   WavePlayer_25
   WavePlayer_26
  ]
  properties {bShouldExcludeFromBranchCulling=true, weights [1,1,1,1,1,1,1,1,1,1,1,1,1]}
 }
 {id=SoundClass_0, children [Mixer_2], properties {soundClassOverride=&Fort_Player_Foley_3P_Enemy_Athena}}
 {id=SoundClass_1, children [Mixer_1], properties {soundClassOverride=&Fort_Player_Foley_1P_Athena}}
 {id=WavePlayer_1, properties {soundWaveAssetPtr=&/Chrome_Blob_Stop_01}}
 {id=WavePlayer_10, properties {soundWaveAssetPtr=&/Chrome_Blob_Stop_10}}
 {id=WavePlayer_11, properties {soundWaveAssetPtr=&/Chrome_Blob_Stop_11}}
 {id=WavePlayer_12, properties {soundWaveAssetPtr=&/Chrome_Blob_Stop_12}}
 {id=WavePlayer_13, properties {soundWaveAssetPtr=&/Chrome_Blob_Stop_13}}
 {id=WavePlayer_14, properties {soundWaveAssetPtr=&Chrome_Blob_Stop_01}}
 {id=WavePlayer_15, properties {soundWaveAssetPtr=&/Chrome_Blob_Stop_02}}
 {id=WavePlayer_16, properties {soundWaveAssetPtr=&/Chrome_Blob_Stop_03}}
 {id=WavePlayer_17, properties {soundWaveAssetPtr=&/Chrome_Blob_Stop_04}}
 {id=WavePlayer_18, properties {soundWaveAssetPtr=&/Chrome_Blob_Stop_05}}
 {id=WavePlayer_19, properties {soundWaveAssetPtr=&/Chrome_Blob_Stop_06}}
 {id=WavePlayer_2, properties {soundWaveAssetPtr=&Chrome_Blob_Stop_02}}
 {id=WavePlayer_20, properties {soundWaveAssetPtr=&/Chrome_Blob_Stop_07}}
 {id=WavePlayer_21, properties {soundWaveAssetPtr=&/Chrome_Blob_Stop_08}}
 {id=WavePlayer_22, properties {soundWaveAssetPtr=&/Chrome_Blob_Stop_09}}
 {id=WavePlayer_23, properties {soundWaveAssetPtr=&Chrome_Blob_Stop_10}}
 {id=WavePlayer_24, properties {soundWaveAssetPtr=&Chrome_Blob_Stop_11}}
 {id=WavePlayer_25, properties {soundWaveAssetPtr=&Chrome_Blob_Stop_12}}
 {id=WavePlayer_26, properties {soundWaveAssetPtr=&Chrome_Blob_Stop_13}}
 {id=WavePlayer_3, properties {soundWaveAssetPtr=&Chrome_Blob_Stop_03}}
 {id=WavePlayer_4, properties {soundWaveAssetPtr=&Chrome_Blob_Stop_04}}
 {id=WavePlayer_5, properties {soundWaveAssetPtr=&Chrome_Blob_Stop_05}}
 {id=WavePlayer_6, properties {soundWaveAssetPtr=&Chrome_Blob_Stop_06}}
 {id=WavePlayer_7, properties {soundWaveAssetPtr=&Chrome_Blob_Stop_07}}
 {id=WavePlayer_8, properties {soundWaveAssetPtr=&Chrome_Blob_Stop_08}}
 {id=WavePlayer_9, properties {soundWaveAssetPtr=&Chrome_Blob_Stop_09}}
]
properties {
 attenuationSettings=&/Chrome_Player_Blob_Stop_Att
 bDebug=true
 bOverrideConcurrency=true
 maxDistance=1500
 priority=0
 volumeMultiplier=2
 concurrencyOverrides {maxCount=2}
}
