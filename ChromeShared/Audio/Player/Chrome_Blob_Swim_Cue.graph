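# Chrome blob swim sound cue for players, varying the mix by listener affiliation.
# A switch on the PlayerAffliation parameter selects a boosted first-person foley branch, shared by two inputs, or one of two third-person branches, one heavily attenuated.
# Both third-person branches use the Enemy foley sound class.
# Each branch randomizes eight swim waves with pitch modulation, under blob attenuation and a concurrency limit.
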SoundCue /ChromeShared/Audio/Player/Chrome_Blob_Swim_Cue
root=Switch_0
nodes [
 {id=Mixer_3, children [Modulator_9], properties {inputVolume [4]}}
 {id=Mixer_4, children [Modulator_12], properties {inputVolume [0.1]}}
 {id=Mixer_5, children [Modulator_13], properties {inputVolume [1]}}
 {id=Modulator_12, children [Random_9], properties {pitchMax=1.2, volumeMax=1}}
 {id=Modulator_13, children [Random_10], properties {pitchMax=1.2, volumeMax=1}}
 {id=Modulator_9, children [Random_6], properties {pitchMax=1.2, volumeMax=1}}
 {
  id=Random_10
  children [WavePlayer_77, WavePlayer_78, WavePlayer_79, WavePlayer_80, WavePlayer_81, WavePlayer_82, WavePlayer_83, WavePlayer_84]
  properties {bShouldExcludeFromBranchCulling=true, weights [1,1,1,1,1,1,1,1]}
 }
 {
  id=Random_6
  children [WavePlayer_0, WavePlayer_46, WavePlayer_47, WavePlayer_48, WavePlayer_49, WavePlayer_50, WavePlayer_51, WavePlayer_52]
  properties {bShouldExcludeFromBranchCulling=true, weights [1,1,1,1,1,1,1,1]}
 }
 {
  id=Random_9
  children [WavePlayer_69, WavePlayer_70, WavePlayer_71, WavePlayer_72, WavePlayer_73, WavePlayer_74, WavePlayer_75, WavePlayer_76]
  properties {bShouldExcludeFromBranchCulling=true, weights [1,1,1,1,1,1,1,1]}
 }
 {id=SoundClass_3, children [Mixer_3], properties {soundClassOverride=&Fort_Player_Foley_1P_Athena}}
 {id=SoundClass_4, children [Mixer_4], properties {soundClassOverride=&Fort_Player_Foley_3P_Enemy_Athena}}
 {id=SoundClass_5, children [Mixer_5], properties {soundClassOverride=&Fort_Player_Foley_3P_Enemy_Athena}}
 {id=Switch_0, children [SoundClass_3,SoundClass_3,SoundClass_4,SoundClass_5], properties {intParameterName=PlayerAffliation}}
 {id=WavePlayer_0, properties {soundWaveAssetPtr=&/Chrome_Blob_Swim_01}}
 {id=WavePlayer_46, properties {soundWaveAssetPtr=&/Chrome_Blob_Swim_02}}
 {id=WavePlayer_47, properties {soundWaveAssetPtr=&/Chrome_Blob_Swim_03}}
 {id=WavePlayer_48, properties {soundWaveAssetPtr=&/Chrome_Blob_Swim_04}}
 {id=WavePlayer_49, properties {soundWaveAssetPtr=&/Chrome_Blob_Swim_05}}
 {id=WavePlayer_50, properties {soundWaveAssetPtr=&/Chrome_Blob_Swim_06}}
 {id=WavePlayer_51, properties {soundWaveAssetPtr=&/Chrome_Blob_Swim_07}}
 {id=WavePlayer_52, properties {soundWaveAssetPtr=&/Chrome_Blob_Swim_08}}
 {id=WavePlayer_69, properties {soundWaveAssetPtr=&Chrome_Blob_Swim_01}}
 {id=WavePlayer_70, properties {soundWaveAssetPtr=&Chrome_Blob_Swim_02}}
 {id=WavePlayer_71, properties {soundWaveAssetPtr=&Chrome_Blob_Swim_03}}
 {id=WavePlayer_72, properties {soundWaveAssetPtr=&Chrome_Blob_Swim_04}}
 {id=WavePlayer_73, properties {soundWaveAssetPtr=&Chrome_Blob_Swim_05}}
 {id=WavePlayer_74, properties {soundWaveAssetPtr=&Chrome_Blob_Swim_06}}
 {id=WavePlayer_75, properties {soundWaveAssetPtr=&Chrome_Blob_Swim_07}}
 {id=WavePlayer_76, properties {soundWaveAssetPtr=&Chrome_Blob_Swim_08}}
 {id=WavePlayer_77, properties {soundWaveAssetPtr=&Chrome_Blob_Swim_01}}
 {id=WavePlayer_78, properties {soundWaveAssetPtr=&Chrome_Blob_Swim_02}}
 {id=WavePlayer_79, properties {soundWaveAssetPtr=&Chrome_Blob_Swim_03}}
 {id=WavePlayer_80, properties {soundWaveAssetPtr=&Chrome_Blob_Swim_04}}
 {id=WavePlayer_81, properties {soundWaveAssetPtr=&Chrome_Blob_Swim_05}}
 {id=WavePlayer_82, properties {soundWaveAssetPtr=&Chrome_Blob_Swim_06}}
 {id=WavePlayer_83, properties {soundWaveAssetPtr=&Chrome_Blob_Swim_07}}
 {id=WavePlayer_84, properties {soundWaveAssetPtr=&Chrome_Blob_Swim_08}}
]
properties {
 attenuationSettings=&/Chrome_Player_Blob_Att
 bDebug=true
 bOverrideConcurrency=true
 maxDistance=8000
 volumeMultiplier=0.7
 concurrencyOverrides {maxCount=3}
}
