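# Sound cue for the player "down enter" sound in Chrome build phasing.
# A switch on the PlayerAffliation parameter selects a first-person branch (shared by two inputs) with wall attenuation, or one of two third-person enemy-class branches with occlusion attenuation.
# Each branch sets its own mixer volume, loudest in first person, over a pitch modulator and random pick from the same four waves.
# A concurrency override caps simultaneous instances, stopping the lowest priority, and virtualization is disabled.
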SoundCue /ChromeShared/Audio/Player/Chrome_Build_Phase_Player_Down_Enter_Cue
root=Switch_1
nodes [
 {id=Attenuation_0, children [SoundClass_1], properties {attenuationSettings=/ChromeShared/Audio/Building/Chrome_Build_Phase_Wall_Att}}
 {
  id=Attenuation_1
  children [SoundClass_0]
  properties {attenuationSettings=&/Chrome_Build_Phase_Player_Occlusion_Att}
 }
 {
  id=Attenuation_2
  children [SoundClass_2]
  properties {attenuationSettings=&Chrome_Build_Phase_Player_Occlusion_Att}
 }
 {id=Mixer_1, children [Modulator_0], properties {inputVolume [3.7]}}
 {id=Mixer_2, children [Modulator_9], properties {inputVolume [0.1]}}
 {id=Mixer_3, children [Modulator_10], properties {inputVolume [1]}}
 {id=Modulator_0, children [Random_1], properties {pitchMax=1.1, volumeMax=1, volumeMin=1}}
 {id=Modulator_10, children [Random_5], properties {pitchMax=1.1, volumeMax=1, volumeMin=1}}
 {id=Modulator_9, children [Random_4], properties {pitchMax=1.1, volumeMax=1, volumeMin=1}}
 {
  id=Random_1
  children [WavePlayer_0,WavePlayer_6,WavePlayer_7,WavePlayer_8]
  properties {bShouldExcludeFromBranchCulling=true, weights [1,1,1,1]}
 }
 {
  id=Random_4
  children [WavePlayer_18, WavePlayer_19, WavePlayer_20, WavePlayer_21]
  properties {bShouldExcludeFromBranchCulling=true, weights [1,1,1,1]}
 }
 {
  id=Random_5
  children [WavePlayer_22, WavePlayer_23, WavePlayer_24, WavePlayer_25]
  properties {bShouldExcludeFromBranchCulling=true, weights [1,1,1,1]}
 }
 {id=SoundClass_0, children [Mixer_2], properties {soundClassOverride=&Fort_Player_Foley_3P_Enemy_Athena}}
 {id=SoundClass_1, children [Mixer_1], properties {soundClassOverride=&Fort_Player_Foley_1P_Athena}}
 {id=SoundClass_2, children [Mixer_3], properties {soundClassOverride=&Fort_Player_Foley_3P_Enemy_Athena}}
 {id=Switch_1, children [Attenuation_0, Attenuation_0, Attenuation_1, Attenuation_2], properties {intParameterName=PlayerAffliation}}
 {id=WavePlayer_0, properties {soundWaveAssetPtr=&/Chrome_Build_Phase_Player_Down_01}}
 {id=WavePlayer_18, properties {soundWaveAssetPtr=&Chrome_Build_Phase_Player_Down_01}}
 {id=WavePlayer_19, properties {soundWaveAssetPtr=&/Chrome_Build_Phase_Player_Down_02}}
 {id=WavePlayer_20, properties {soundWaveAssetPtr=&/Chrome_Build_Phase_Player_Down_03}}
 {id=WavePlayer_21, properties {soundWaveAssetPtr=&/Chrome_Build_Phase_Player_Down_04}}
 {id=WavePlayer_22, properties {soundWaveAssetPtr=&Chrome_Build_Phase_Player_Down_01}}
 {id=WavePlayer_23, properties {soundWaveAssetPtr=&Chrome_Build_Phase_Player_Down_02}}
 {id=WavePlayer_24, properties {soundWaveAssetPtr=&Chrome_Build_Phase_Player_Down_03}}
 {id=WavePlayer_25, properties {soundWaveAssetPtr=&Chrome_Build_Phase_Player_Down_04}}
 {id=WavePlayer_6, properties {soundWaveAssetPtr=&Chrome_Build_Phase_Player_Down_02}}
 {id=WavePlayer_7, properties {soundWaveAssetPtr=&Chrome_Build_Phase_Player_Down_03}}
 {id=WavePlayer_8, properties {soundWaveAssetPtr=&Chrome_Build_Phase_Player_Down_04}}
]
properties {
 bOverrideConcurrency=true
 maxDistance=8000
 virtualizationMode=Disabled
 volumeMultiplier=0.8913
 concurrencyOverrides {maxCount=3, resolutionRule=StopLowestPriority}
}
