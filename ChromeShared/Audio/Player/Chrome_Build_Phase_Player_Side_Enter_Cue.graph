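# Sound cue for the Chrome build-phase player-side enter sound, varied by the player's affiliation.
# An integer switch on PlayerAffliation shares a first-person branch with wall attenuation across two inputs, plus two delayed branches with occlusion attenuation.
# Both delayed branches use the third-person enemy foley class at different mix volumes.
# Each branch randomizes the same four waves with pitch modulation, and concurrency limits stop the lowest-priority instance.
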SoundCue /ChromeShared/Audio/Player/Chrome_Build_Phase_Player_Side_Enter_Cue
root=Switch_1
nodes [
 {
  id=Attenuation_0
  children [SoundClass_0]
  properties {attenuationSettings=&/Chrome_Build_Phase_Player_Occlusion_Att}
 }
 {id=Attenuation_1, children [SoundClass_1], properties {attenuationSettings=/ChromeShared/Audio/Building/Chrome_Build_Phase_Wall_Att}}
 {
  id=Attenuation_2
  children [SoundClass_2]
  properties {attenuationSettings=&Chrome_Build_Phase_Player_Occlusion_Att}
 }
 {id=Delay_0, children [Attenuation_2], properties {delayMax=0.58, delayMin=0.58}}
 {id=Delay_1, children [Attenuation_0], properties {delayMax=0.58, delayMin=0.58}}
 {id=Delay_2, children [Attenuation_1]}
 {id=Mixer_1, children [Modulator_0], properties {inputVolume [3.7]}}
 {id=Mixer_2, children [Modulator_4], properties {inputVolume [0.1]}}
 {id=Mixer_3, children [Modulator_5], properties {inputVolume [1]}}
 {id=Modulator_0, children [Random_1], properties {pitchMax=1.1, volumeMax=1, volumeMin=1}}
 {id=Modulator_4, children [Random_6], properties {pitchMax=1.1, volumeMax=1, volumeMin=1}}
 {id=Modulator_5, children [Random_7], properties {pitchMax=1.1, volumeMax=1, volumeMin=1}}
 {
  id=Random_1
  children [WavePlayer_0,WavePlayer_6,WavePlayer_7,WavePlayer_8]
  properties {bShouldExcludeFromBranchCulling=true, weights [1,1,1,1]}
 }
 {
  id=Random_6
  children [WavePlayer_16, WavePlayer_17, WavePlayer_26, WavePlayer_27]
  properties {bShouldExcludeFromBranchCulling=true, weights [1,1,1,1]}
 }
 {
  id=Random_7
  children [WavePlayer_28, WavePlayer_29, WavePlayer_30, WavePlayer_31]
  properties {bShouldExcludeFromBranchCulling=true, weights [1,1,1,1]}
 }
 {id=SoundClass_0, children [Mixer_2], properties {soundClassOverride=&Fort_Player_Foley_3P_Enemy_Athena}}
 {id=SoundClass_1, children [Mixer_1], properties {soundClassOverride=&Fort_Player_Foley_1P_Athena}}
 {id=SoundClass_2, children [Mixer_3], properties {soundClassOverride=&Fort_Player_Foley_3P_Enemy_Athena}}
 {id=Switch_1, children [Delay_2,Delay_2,Delay_1,Delay_0], properties {intParameterName=PlayerAffliation}}
 {id=WavePlayer_0, properties {soundWaveAssetPtr=&/Chrome_Build_Phase_Player_Side_01}}
 {id=WavePlayer_16, properties {soundWaveAssetPtr=&Chrome_Build_Phase_Player_Side_01}}
 {id=WavePlayer_17, properties {soundWaveAssetPtr=&/Chrome_Build_Phase_Player_Side_02}}
 {id=WavePlayer_26, properties {soundWaveAssetPtr=&/Chrome_Build_Phase_Player_Side_03}}
 {id=WavePlayer_27, properties {soundWaveAssetPtr=&/Chrome_Build_Phase_Player_Side_04}}
 {id=WavePlayer_28, properties {soundWaveAssetPtr=&Chrome_Build_Phase_Player_Side_01}}
 {id=WavePlayer_29, properties {soundWaveAssetPtr=&Chrome_Build_Phase_Player_Side_02}}
 {id=WavePlayer_30, properties {soundWaveAssetPtr=&Chrome_Build_Phase_Player_Side_03}}
 {id=WavePlayer_31, properties {soundWaveAssetPtr=&Chrome_Build_Phase_Player_Side_04}}
 {id=WavePlayer_6, properties {soundWaveAssetPtr=&Chrome_Build_Phase_Player_Side_02}}
 {id=WavePlayer_7, properties {soundWaveAssetPtr=&Chrome_Build_Phase_Player_Side_03}}
 {id=WavePlayer_8, properties {soundWaveAssetPtr=&Chrome_Build_Phase_Player_Side_04}}
]
properties {
 bOverrideConcurrency=true
 maxDistance=8000
 virtualizationMode=Disabled
 volumeMultiplier=0.8913
 concurrencyOverrides {maxCount=3, resolutionRule=StopLowestPriority}
}
