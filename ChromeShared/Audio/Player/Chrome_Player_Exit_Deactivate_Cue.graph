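# Sound cue for the Chrome player exit or deactivate sound, routed to the Fort_Gadgets_Items sound class.
# A single wave player feeds the Chrome_Player_Deactivate_01 wave straight to the output, with no randomization or modulation.
# Its own concurrency override caps it at one instance with releasable volume scaling, and virtualization is disabled.
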SoundCue /ChromeShared/Audio/Player/Chrome_Player_Exit_Deactivate_Cue
root=WavePlayer_0
nodes [{id=WavePlayer_0, properties {soundWaveAssetPtr=&/Chrome_Player_Deactivate_01}}]
properties {
 bOverrideConcurrency=true
 maxDistance=2097152
 soundClassObject=&Fort_Gadgets_Items
 virtualizationMode=Disabled
 volumeMultiplier=1
 concurrencyOverrides {bVolumeScaleCanRelease=true, maxCount=1, volumeScale=0.5, volumeScaleAttackTime=0.1}
}
