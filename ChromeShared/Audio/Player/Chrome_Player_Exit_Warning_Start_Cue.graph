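# Sound cue for the start of the Chrome player exit warning, playing a timer start sound.
# A single wave player feeds the root directly, with no randomization or modulation.
# It uses the gadget items sound class and overrides concurrency to one instance, resolved by stopping the lowest priority sound.
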SoundCue /ChromeShared/Audio/Player/Chrome_Player_Exit_Warning_Start_Cue
root=WavePlayer_0
nodes [{id=WavePlayer_0, properties {soundWaveAssetPtr=&/Chrome_Player_Timer_Start_01}}]
properties {
 bOverrideConcurrency=true
 maxDistance=2097152
 soundClassObject=&Fort_Gadgets_Items
 volumeMultiplier=1
 concurrencyOverrides {maxCount=1, resolutionRule=StopLowestPriority}
}
