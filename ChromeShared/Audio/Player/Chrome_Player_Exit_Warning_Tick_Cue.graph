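# Sound cue for the Chrome player's exit warning tick, built from timer tick samples.
# A root Random node picks one of five equally weighted tick waves on each play and is excluded from branch culling.
# It plays in the gadgets and items sound class, and its own concurrency override allows two instances, stopping the oldest when exceeded.
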SoundCue /ChromeShared/Audio/Player/Chrome_Player_Exit_Warning_Tick_Cue
root=Random_0
nodes [
 {
  id=Random_0
  children [WavePlayer_1,WavePlayer_2,WavePlayer_3,WavePlayer_4,WavePlayer_5]
  properties {bShouldExcludeFromBranchCulling=true, weights [1,1,1,1,1]}
 }
 {id=WavePlayer_1, properties {soundWaveAssetPtr=&/Chrome_Player_Timer_Tick_01}}
 {id=WavePlayer_2, properties {soundWaveAssetPtr=&/Chrome_Player_Timer_Tick_02}}
 {id=WavePlayer_3, properties {soundWaveAssetPtr=&/Chrome_Player_Timer_Tick_03}}
 {id=WavePlayer_4, properties {soundWaveAssetPtr=&/Chrome_Player_Timer_Tick_04}}
 {id=WavePlayer_5, properties {soundWaveAssetPtr=&/Chrome_Player_Timer_Tick_05}}
]
properties {
 bOverrideConcurrency=true
 maxDistance=2097152
 soundClassObject=&Fort_Gadgets_Items
 volumeMultiplier=1
 concurrencyOverrides {maxCount=2, resolutionRule=StopOldest}
}
