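# Sound cue for the Chrome player heal start sound, played quietly in the gadgets and items sound class.
# A Random node picks one of four heal-start wave variations with equal weight and is excluded from branch culling, then feeds a Modulator at the root.
# It uses the Chrome player status effects attenuation and overrides concurrency to allow a single instance, stopping the lowest priority.
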SoundCue /ChromeShared/Audio/Player/Chrome_Player_Heal_Cue
root=Modulator_0
nodes [
 {id=Modulator_0, children [Random_0]}
 {
  id=Random_0
  children [WavePlayer_1,WavePlayer_2,WavePlayer_3,WavePlayer_4]
  properties {bShouldExcludeFromBranchCulling=true, weights [1,1,1,1]}
 }
 {id=WavePlayer_1, properties {soundWaveAssetPtr=&/Chrome_Player_Heal_Start_04}}
 {id=WavePlayer_2, properties {soundWaveAssetPtr=&/Chrome_Player_Heal_Start_01}}
 {id=WavePlayer_3, properties {soundWaveAssetPtr=&/Chrome_Player_Heal_Start_02}}
 {id=WavePlayer_4, properties {soundWaveAssetPtr=&/Chrome_Player_Heal_Start_03}}
]
properties {
 attenuationSettings=&/Chrome_Player_Status_Effects_Att
 bOverrideConcurrency=true
 maxDistance=3500
 soundClassObject=&Fort_Gadgets_Items
 volumeMultiplier=0.3
 concurrencyOverrides {maxCount=1, resolutionRule=StopLowestPriority}
}
