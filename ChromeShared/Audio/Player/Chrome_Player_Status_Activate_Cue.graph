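# Sound cue for the Chrome player status activation, playing one of four random, modulated application sounds.
# A branch on IsLocal routes local playback to the first-person foley class with heavily boosted volume, and otherwise to the enemy third-person foley class with attenuation and much lower volume.
# Both paths draw on the same four waves.
# A concurrency override limits it to one instance, and virtualization is disabled.
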SoundCue /ChromeShared/Audio/Player/Chrome_Player_Status_Activate_Cue
root=Branch_0
nodes [
 {id=Attenuation_0, children [Mixer_2], properties {attenuationSettings=&/Chrome_Player_Status_Effects_Att}}
 {id=Branch_0, children [SoundClass_1,SoundClass_0,SoundClass_0], properties {boolParameterName=IsLocal}}
 {id=Mixer_1, children [Modulator_0], properties {inputVolume [6]}}
 {id=Mixer_2, children [Modulator_2], properties {inputVolume [0.1]}}
 {id=Modulator_0, children [Random_0]}
 {id=Modulator_2, children [Random_2]}
 {id=Random_0, children [WavePlayer_1,WavePlayer_2,WavePlayer_3,WavePlayer_4], properties {weights [1,1,1,1]}}
 {id=Random_2, children [WavePlayer_5, WavePlayer_9, WavePlayer_10, WavePlayer_11], properties {weights [1,1,1,1]}}
 {
  id=SoundClass_0
  children [Attenuation_0]
  properties {soundClassOverride=&Fort_Player_Foley_3P_Enemy_Athena}
 }
 {id=SoundClass_1, children [Mixer_1], properties {soundClassOverride=&Fort_Player_Foley_1P_Athena}}
 {id=WavePlayer_1, properties {soundWaveAssetPtr=&/Chrome_Player_Application_01}}
 {id=WavePlayer_10, properties {soundWaveAssetPtr=&/Chrome_Player_Application_03}}
 {id=WavePlayer_11, properties {soundWaveAssetPtr=&/Chrome_Player_Application_04}}
 {id=WavePlayer_2, properties {soundWaveAssetPtr=&/Chrome_Player_Application_02}}
 {id=WavePlayer_3, properties {soundWaveAssetPtr=&Chrome_Player_Application_03}}
 {id=WavePlayer_4, properties {soundWaveAssetPtr=&Chrome_Player_Application_04}}
 {id=WavePlayer_5, properties {soundWaveAssetPtr=&Chrome_Player_Application_01}}
 {id=WavePlayer_9, properties {soundWaveAssetPtr=&Chrome_Player_Application_02}}
]
properties {
 bOverrideConcurrency=true
 maxDistance=3500
 virtualizationMode=Disabled
 volumeMultiplier=1
 concurrencyOverrides {bVolumeScaleCanRelease=true, maxCount=1, volumeScale=0.5, volumeScaleAttackTime=0.1}
}
