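# Sound cue for the shatter sound played when the Chrome status is removed from a player.
# An IsLocal branch sends the local player to a louder, non-attenuated first-person foley class path, and everyone else, including the unset case, to a quiet, attenuated third-person enemy class path.
# Each path modulates a random pick of three shatter waves.
# Concurrency allows one instance, and virtualization is disabled.
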SoundCue /ChromeShared/Audio/Player/Chrome_Player_Status_Removal_Shattered_Cue
root=Branch_1
nodes [
 {id=Attenuation_0, children [Mixer_2], properties {attenuationSettings=&/Chrome_Player_Status_Effects_Att}}
 {id=Branch_1, children [SoundClass_1,SoundClass_0,SoundClass_0], properties {boolParameterName=IsLocal}}
 {id=Mixer_1, children [Modulator_0], properties {inputVolume [4]}}
 {id=Mixer_2, children [Modulator_2], properties {inputVolume [0.1]}}
 {id=Modulator_0, children [Random_1]}
 {id=Modulator_2, children [Random_2]}
 {id=Random_1, children [WavePlayer_0,WavePlayer_5,WavePlayer_6], properties {bShouldExcludeFromBranchCulling=true, weights [1,1,1]}}
 {id=Random_2, children [WavePlayer_4,WavePlayer_7,WavePlayer_9], properties {bShouldExcludeFromBranchCulling=true, weights [1,1,1]}}
 {
  id=SoundClass_0
  children [Attenuation_0]
  properties {soundClassOverride=&Fort_Player_Foley_3P_Enemy_Athena}
 }
 {id=SoundClass_1, children [Mixer_1], properties {soundClassOverride=&Fort_Player_Foley_1P_Athena}}
 {id=WavePlayer_0, properties {soundWaveAssetPtr=&/Chrome_Player_Removal_Shatter_01}}
 {id=WavePlayer_4, properties {soundWaveAssetPtr=&Chrome_Player_Removal_Shatter_01}}
 {id=WavePlayer_5, properties {soundWaveAssetPtr=&/Chrome_Player_Removal_Shatter_02}}
 {id=WavePlayer_6, properties {soundWaveAssetPtr=&/Chrome_Player_Removal_Shatter_03}}
 {id=WavePlayer_7, properties {soundWaveAssetPtr=&Chrome_Player_Removal_Shatter_02}}
 {id=WavePlayer_9, properties {soundWaveAssetPtr=&Chrome_Player_Removal_Shatter_03}}
]
properties {
 bOverrideConcurrency=true
 maxDistance=3500
 virtualizationMode=Disabled
 volumeMultiplier=0.3
 concurrencyOverrides {bVolumeScaleCanRelease=true, maxCount=1, volumeScale=0.5, volumeScaleAttackTime=0.1}
}
